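# M_Athena_Island_Grey is a Default Lit material in the Athena landscape folder, named as a grey island surface.
# It holds only two nodes, a constant and a scalar Brightness parameter with a low default, with no textures and no visible wiring to material outputs.
# The material is flagged for use on spline meshes and uses legacy pixel depth offset mode.
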Material /FigmentEnvironment/LM_8_1_1/Athena/Environments/Landscape/Material/M_Athena_Island_Grey
graph {
 properties {bUsedWithSplineMeshes=true, pixelDepthOffsetMode=PDOM_Legacy, usedShadingModels=Default Lit}
 nodes [
  Constant() {r=0.95}
  ScalarParameter() {defaultValue=0.035, parameterName=Brightness}
 ]
}
