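# Jungle sand material for meshes, including spline meshes, with terrain-matched base sand, optional grass and a final RVT blending function.
# Vertex blue blends the shared arid sand function into a tinted, world-aligned beach sand layer whose texture alpha drives roughness and detail-normal strength.
# Vertex red height-blends grass on top, optionally biome color-matched.
# Normals shift between world and tangent space between layers, as the material outputs world-space normals.
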Material /FigmentEnvironment/LM_8_1_1/Athena/Environments/Landscape/Material/M_Athena_JungleSand
graph {
 properties {
  bTangentSpaceNormal=false
  bUseMaterialAttributes=true
  bUsedWithSplineMeshes=true
  pixelDepthOffsetMode=PDOM_Legacy
  usedShadingModels=Default Lit
 }
 nodes [
  n1 = VertexColor()
  n2 = TextureObjectParameter() {
   group=Mud
   parameterName=Mud_Diffuse
   samplerSource=SSM_Wrap_WorldGroupSettings
   texture=/FigmentEnvironment/LM_8_1_1/Athena/Environments/Landscape/Textures/T_Athena_BeachSand_01
  }
  n3 = ScalarParameter() {defaultValue=900, group=Mud, parameterName=MudTextureTilingFrequency}
  n4 = MaterialFunctionCall(TextureObject=@n2, TextureSize=@n3, Export Float 4=StaticBool() {value=true}) {materialFunction=&WorldAlignedTexture}
  n5 = ComponentMask(None=@n4.Z Texture) {a=true}
  MaterialFunctionCall(A=@n1.B, Transition Phase=@n1.B, Height Texture=@n5, Contrast=Constant()) {materialFunction=&HeightLerp}
  n6 = MaterialFunctionCall!(In, UseRoughness) {materialFunction=/FigmentEnvironment/LM_8_1_1/Athena/Environments/Landscape/MaterialFunctions/Arid/MF_Athena_Arid_Sand}
  n7 = Constant() {desc=was 0.1, r=0.095}
  n8 = ComponentMask(None=@n6.BaseColor_NoPebbles) {r=true}
  n9 = VertexNormalWS()
  n10 = SetMaterialAttributes!(
   MaterialAttributes
   Base Color=@n6.BaseColor_NoPebbles
   Specular=FeatureLevelSwitch(Default=@n7, ES2=@n8, ES3_1=@n8, SM4_REMOVED=@n7, SM5=@n7, SM6=@n7) {desc=Match Terrain}
   Roughness=Constant() {desc=was 0.95, r=0.9}
   Normal=FeatureLevelSwitch(Default=@n6.Normal, ES2=@n9, ES3_1=@n9, SM4_REMOVED=@n9, SM5=@n6.Normal, SM6=@n6.Normal) {desc=Match Terrain}
  ) {attributeSetTypes [
    #1
    #2
    #3
    #4
   ]}
  n11 = Multiply(
   A=VectorParameter() {group=Mud, parameterName=MudTint, defaultValue=<rgba 0.8333 0.8192 0.7533 0>}.RGB
   B=ComponentMask(None=@n4.Z Texture) {b=true, g=true, r=true}
  )
  n12 = ScalarParameter() {defaultValue=0.1, parameterName=Sand Spec}
  n13 = LinearInterpolate(
   A=ScalarParameter() {defaultValue=0.25, parameterName=Sand Roughness}
   B=ScalarParameter() {defaultValue=0.5, parameterName=Sand Roughness Max}
   Alpha=@n5
  )
  n14 = Divide(
   A=ComponentMask(None=WorldPosition().XYZ) {g=true, r=true}
   B=ScalarParameter() {defaultValue=512, group=Mud, parameterName=SandNormalTilingFrequency}
  )
  n15 = TextureObjectParameter() {
   group=Mud
   parameterName=Mud_Normal
   samplerSource=SSM_Wrap_WorldGroupSettings
   samplerType=SAMPLERTYPE_Normal
   texture=/FigmentEnvironment/LM_8_1_1/Packages/Fortress_SharedMaterials/Textures/FlatNormalBlue
  }
  n16 = MaterialFunctionCall(TextureObject=@n15, TextureSize=@n3) {materialFunction=&WorldAlignedNormal}
  n17 = MaterialFunctionCall(
   BaseNormal=@n16.Z Texture
   AdditionalNormal=LinearInterpolate(
    A=TextureSample(UVs=@n14) {samplerType=SAMPLERTYPE_Normal, texture=/FigmentEnvironment/LM_8_1_1/Packages/DS_Fortnight/T_MudBeachNormal_01}.RGB
    B=Constant3Vector() {constant=<rgba 0 0 1 0>}
    Alpha=@n5
   )
  ) {materialFunction=&BlendAngleCorrectedNormals}
  n18 = QualitySwitch(Default=@n17.Result, Low=@n16.Z Texture, High=@n17.Result, Medium=@n17.Result)
  n19 = GetMaterialAttributes(None=MaterialFunctionCall!(
    Base Material=@n10
    Top Material=MakeMaterialAttributes!(
     BaseColor=@n11
     Metallic
     Specular=@n12
     Roughness=@n13
     Anisotropy
     EmissiveColor
     Opacity
     OpacityMask
     Normal=@n18
     Tangent
     WorldPositionOffset
     SubsurfaceColor
     ClearCoat
     ClearCoatRoughness
     AmbientOcclusion
     Refraction
     CustomizedUVs_0
     CustomizedUVs_1
     CustomizedUVs_2
     CustomizedUVs_3
     CustomizedUVs_4
     CustomizedUVs_5
     CustomizedUVs_6
     CustomizedUVs_7
     PixelDepthOffset
     ShadingModel
     Displacement
    )
    Alpha=@n1.B
   ) {materialFunction=/Engine/Functions/MaterialLayerFunctions/MatLayerBlend_Standard}.Blended Material) {attributeGetTypes [#4]}
  n20 = SetMaterialAttributes!(
   MaterialAttributes=@n19.MaterialAttributes
   Normal=Transform(None=@n19.Normal) {transformSourceType=TRANSFORMSOURCE_World, transformType=TRANSFORM_Tangent}
  ) {attributeSetTypes [#4]}
  n21 = MaterialFunctionCall() {materialFunction=/FigmentEnvironment/LM_8_1_1/Athena/Environments/Landscape/MaterialFunctions/Standard/MF_Athena_Grass_01}
  n22 = MaterialFunctionCall!(Material=@n21.Result) {materialFunction=/FigmentEnvironment/LM_8_1_1/Athena/Environments/Materials/MF_StaticMeshBiomeColor}
  n23 = VertexColor()
  n24 = Constant() {r=1}
  n25 = StaticSwitchParameter(
   True=MaterialFunctionCall(
    Base Material=@n20
    Top Material=StaticSwitchParameter(True=@n22.Material, False=@n21.Result) {defaultValue=true, parameterName=UseGrassColorMatch}
    Alpha=MaterialFunctionCall(Transition Phase=@n23.R, Height Texture=@n21.Height, Contrast=@n24) {materialFunction=&HeightLerp}.Alpha
   ) {materialFunction=&MatLayerBlend_Standard}.Blended Material
   False=@n20
  ) {defaultValue=true, parameterName=UseGrass}
  n26 = GetMaterialAttributes(None=@n25) {attributeGetTypes [#4]}
  n27 = Transform(None=@n26.Normal)
  MaterialFunctionCall!(In=StaticSwitchParameter(
    True=@n25
    False=SetMaterialAttributes(MaterialAttributes=@n26.MaterialAttributes, Normal=@n27) {attributeSetTypes [#4]}
   ) {parameterName=UseWorldSpaceNormal}) {materialFunction=/FigmentEnvironment/Foliage/MF_Figment_RVTBlendingOn}
 ]
}
