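# Landscape material function that adds crater burn damage around a ground-zero point, either world-space or object-relative.
# A projected crater texture darkens base color and specular in streaks, with vertex blue restoring the originals.
# Two panned lava samples, masked by noise, streaks and normals facing ground zero, form animated emissive fire.
# The Burn Location Local switch skips darkening and confines fire to the radial falloff, which otherwise adds core glow.
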MaterialFunction /Game/Athena/Environments/Landscape/MaterialFunctions/MF_DamageBurn
graph {nodes [
  FunctionInput() {inputName=Noise, inputType=FunctionInput_Scalar}
  n1 = GetMaterialAttributes(None=FunctionInput() {inputType=FunctionInput_MaterialAttributes}) {attributeGetTypes [#1, #2, #3]}
  n2 = VectorParameter() {parameterName=DrakenStreks, defaultValue=<rgba 0.0833 0.0833 0.0833 1>}
  n3 = VectorParameter() {parameterName=LocationGroundzero, defaultValue=<rgba -555.25 640 -800 1>}
  n4 = Add(A=ObjectPositionWS(), B=@n3.RGB)
  n5 = Subtract(
   A=WorldPosition().XYZ
   B=StaticSwitchParameter(True=@n4, False=StaticSwitchParameter(True=@n4, False=@n3.RGB) {parameterName=Burn Location Local}) {parameterName=LocaolPosition}
  )
  n6 = ComponentMask(None=@n5) {g=true, r=true}
  n7 = Subtract(A=Divide(A=@n6, B=ScalarParameter() {defaultValue=14058.2656, parameterName=CraterSize}), B=Constant2Vector() {g=0.5, r=0.5})
  n8 = Saturate(None=LinearInterpolate(
    A=ScalarParameter() {defaultValue=-1.4438, parameterName=StreaksContrast_1}
    B=ScalarParameter() {defaultValue=5.1737, parameterName=StreaksContrast_2}
    Alpha=TextureSample(UVs=@n7) {samplerType=SAMPLERTYPE_Masks, texture=/Game/Athena/Environments/Textures/Crater_Small_S}.R
   ))
  n9 = StaticSwitchParameter(True=@n1.BaseColor, False=LinearInterpolate(A=Multiply(A=@n1.BaseColor, B=@n2.RGB), B=@n1.BaseColor, Alpha=@n8)) {parameterName=Burn Location Local}
  n10 = VertexColor()
  n11 = Divide(A=Distance(A=@n6, B=Constant2Vector()), B=ScalarParameter() {defaultValue=2237.1611, parameterName=RadiusGLow})
  n12 = Power(Base=OneMinus(None=Saturate(None=@n11))) {constExponent=3}
  n13 = WorldPosition()
  n14 = TextureSample(UVs=Divide(A=ComponentMask(None=@n13.XYZ) {g=true, r=true}, B=ScalarParameter() {defaultValue=269.0455, parameterName=SizeNoise2})) {samplerType=SAMPLERTYPE_LinearColor, texture=/Game/Global/Textures/Default/Noise/T_Noise01_Inverse}
  n15 = MaterialFunctionCall(In=@n14.R, Contrast=ScalarParameter() {defaultValue=8, parameterName=ContrastNoise2}) {materialFunction=&CheapContrast}
  n16 = Normalize(VectorInput=@n5)
  n17 = Normalize(VectorInput=@n1.Normal)
  n18 = Power(Base=OneMinus(None=DotProduct(A=@n16, B=@n17)), Exp=ScalarParameter() {defaultValue=3, parameterName=CharedPower})
  n19 = Divide(A=ComponentMask(None=WorldPosition().XYZ) {g=true, r=true}, B=ScalarParameter() {defaultValue=97.0664, parameterName=FireSize})
  n20 = ScalarParameter() {defaultValue=0.2, parameterName=FireSpeed}
  n21 = OneMinus(None=TextureSample(UVs=Panner(Coordinate=@n19, Speed=@n20)) {samplerType=SAMPLERTYPE_Masks, texture=/Game/Environments/Sets/Vulcano/Textures/LavaGround_M}.G)
  n22 = Multiply(A=@n20, B=Constant2Vector() {g=-0.9, r=-1.2})
  n23 = OneMinus(None=TextureSample(UVs=Panner(Coordinate=@n19, Speed=@n22)) {samplerType=SAMPLERTYPE_Masks, texture=&LavaGround_M}.G)
  n24 = Add(A=@n21, B=@n23)
  n25 = Multiply(A=Clamp(None=@n18), B=@n24)
  n26 = ScalarParameter() {defaultValue=8, parameterName=FireContrast}
  n27 = Multiply(
   A=@n25
   B=LinearInterpolate(
    A=VectorParameter() {parameterName=FireGlow, defaultValue=<rgba 0.1771 0.0274 0 1>}.RGB
    B=VectorParameter() {parameterName=FireGlow2, defaultValue=<rgba 1 0.1825 0.0573 1>}.RGB
    Alpha=Power(Base=@n25, Exp=@n26)
   )
  )
  n28 = Multiply(A=Multiply(A=Multiply(A=@n15.Result, B=@n27), B=@n18), B=OneMinus(None=@n8))
  n29 = Multiply(A=@n12, B=@n28)
  n30 = Add(A=Multiply(A=VectorParameter() {parameterName=GllowCore, defaultValue=<rgba 0.4062 0.0716 0 1>}.RGB, B=@n12), B=@n28)
  n31 = ScalarParameter() {defaultValue=0.4095, parameterName=SpecularDark}
  n32 = StaticSwitchParameter(True=@n1.Specular, False=Multiply(A=@n1.Specular, B=LinearInterpolate(A=@n31, Alpha=@n8))) {parameterName=Burn Location Local}
  FunctionOutput(None=SetMaterialAttributes!(
    MaterialAttributes=@n1.MaterialAttributes
    BaseColor=LinearInterpolate(A=@n9, B=@n1.BaseColor, Alpha=@n10.B)
    EmissiveColor=StaticSwitchParameter(True=@n29, False=@n30) {parameterName=Burn Location Local}
    Specular=LinearInterpolate(A=@n32, B=@n1.Specular, Alpha=@n10.B)
   ) {attributeSetTypes [#1, #4, #3]})
  MaterialFunctionCall(Height Texture=@n24, Contrast=ScalarParameter() {defaultValue=0.1619, parameterName=ContrastFireTransition}) {materialFunction=&HeightLerp}
  Multiply(A=@n21, B=@n23)
  OneMinus(None=@n14.R)
  TextureCoordinate()
  Transform(None=@n1.Normal)
 ]}
