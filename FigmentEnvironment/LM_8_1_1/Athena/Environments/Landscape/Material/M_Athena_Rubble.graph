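# Nanite-enabled Athena landscape-environment material for rubble piles with soil, projected from world XY.
# Two differently scaled, tinted rubble layers, bump-offset on deferred paths, are height-blended, then soil is layered on using noise, rubble height and vertex blue.
# An optional damage-burn function, on by default, follows, and normals are transformed for world-space output.
# Unless skipped, non-mobile deferred paths blend in the landscape runtime virtual texture by height above its surface.
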Material /FigmentEnvironment/LM_8_1_1/Athena/Environments/Landscape/Material/M_Athena_Rubble
graph {
 properties {
  bTangentSpaceNormal=false
  bUseMaterialAttributes=true
  bUsedWithNanite=true
  pixelDepthOffsetMode=PDOM_Legacy
  usedShadingModels=Default Lit
 }
 nodes [
  n1 = ComponentMask(None=WorldPosition().XYZ) {g=true, r=true}
  n2 = Divide(A=@n1, B=ScalarParameter() {defaultValue=-900, parameterName=SizeSoil})
  n3 = TextureSample(UVs=@n2) {samplerType=SAMPLERTYPE_Normal, texture=/Game/Packages/DS_Fortnight/WhiteRockCustom_N}
  MaterialFunctionCall(XYZ=@n3.RGB) {materialFunction=&Swizzle}
  n4 = Divide(A=@n1, B=ScalarParameter() {defaultValue=2034.6349, parameterName=SIzeRubble2})
  n5 = MaterialFunctionCall!(
   HeightTexture=TextureSample(UVs=@n4) {samplerType=SAMPLERTYPE_Masks, texture=/Game/Environments/Sets/Rubble/Props/Textures/T_Rubble_Pile_H}.R
   TextureCoordinates=@n4
   ReferencePlane=ScalarParameter() {defaultValue=0.5, parameterName=RefPlane2}
   HeightRatioInput=ScalarParameter() {defaultValue=0.007, parameterName=BumpOffset2}
   CameraVector=CameraVectorWS()
  ) {materialFunction=/Engine/Functions/Engine_MaterialFunctions02/Utility/BumpOffset_Advanced.BumpOffset_advanced}
  n6 = ShadingPathSwitch(Default=@n5.UVs, Deferred=@n5.UVs, Forward=@n4, Mobile=@n4)
  n7 = TextureSample(UVs=@n6) {samplerType=SAMPLERTYPE_Normal, texture=/Game/Environments/Sets/Rubble/Props/Textures/T_rubble_pile_N}
  Multiply(A=@n7.RGB, B=Constant3Vector() {constant=<rgba -1 1 1 0>})
  n8 = Divide(A=@n1, B=ScalarParameter() {defaultValue=800, parameterName=SIzeRubble})
  n9 = MaterialFunctionCall(
   HeightTexture=TextureSample(UVs=@n8) {samplerType=SAMPLERTYPE_Masks, texture=&T_Rubble_Pile_H}.R
   TextureCoordinates=@n8
   ReferencePlane=ScalarParameter() {defaultValue=0.5, parameterName=RefPlane1}
   HeightRatioInput=ScalarParameter() {defaultValue=0.02, parameterName=BumpOffset1}
   CameraVector=CameraVectorWS()
  ) {materialFunction=&BumpOffset_advanced}
  n10 = ShadingPathSwitch(Default=@n9.UVs, Deferred=@n9.UVs, Forward=@n8, Mobile=@n8)
  n11 = TextureSample(UVs=@n10) {samplerType=SAMPLERTYPE_Normal, texture=&T_rubble_pile_N}
  Multiply(A=@n11.RGB, B=Constant3Vector() {constant=<rgba -1 1 1 0>})
  PinBase()
  n12 = Subtract(A=WorldPosition().XYZ, B=ActorPositionWS())
  n13 = ScalarParameter() {defaultValue=-546.6205, group=Ground, parameterName=Bottom _HeightOffset, sortPriority=6}
  n14 = ComponentMask(None=Divide(A=Subtract(A=@n12, B=@n13), B=ScalarParameter() {defaultValue=-183.6973, group=Ground, parameterName=Bottom_Blur, sortPriority=7})) {b=true}
  n15 = VertexNormalWS()
  n16 = Saturate(None=Multiply(A=@n14, B=ComponentMask(None=@n15) {b=true}))
  n17 = Saturate(None=@n14)
  PinBase!(Fade *Up=@n16, Fade Only=@n17) {reroutePins [
    Fade *Up {expression=&self:MaterialExpressionReroute_2}
    Fade Only {expression=&self:MaterialExpressionReroute_3}
   ]}
  MaterialExpressionReroute_2 = Reroute(None=@n16)
  MaterialExpressionReroute_3 = Reroute(None=@n17)
  n18 = TextureSample(UVs=@n6) {samplerType=SAMPLERTYPE_Masks, texture=&T_Rubble_Pile_H}
  Reroute(None=@n18.R)
  n19 = RuntimeVirtualTextureSampleParameter() {
   group=RVT
   materialType=BaseColor_Normal_Specular_YCoCg
   parameterName=RVT
   sortPriority=3
   virtualTexture=/FigmentCore/Game/Athena/Environments/Materials/RVT/Athena_RuntimeVirtualTexture
  }
  n20 = SetMaterialAttributes!(MaterialAttributes, Base Color=@n19.BaseColor, Specular=@n19.BaseColor, Roughness=@n19.BaseColor, Normal=@n19.BaseColor) {attributeSetTypes [
    #1
    #2
    #3
    #4
   ]}
  n21 = WorldPosition()
  n22 = RuntimeVirtualTextureSampleParameter() {
   group=RVT
   materialType=WorldHeight
   parameterName=RVTHeight
   sortPriority=4
   virtualTexture=/FigmentCore/Game/Athena/Environments/Materials/RVT/Athena_RuntimeVirtualTexture_Height
  }
  n23 = Subtract(A=Add(A=ScalarParameter() {defaultValue=450, parameterName=RVT Z Offset}, B=ComponentMask(None=@n21.XYZ) {b=true}), B=@n22.WorldHeight)
  n24 = Saturate(None=Divide(
    A=NamedRerouteDeclaration(None=@n23) {name=RVT Height, nodeColor=<rgba 1 0.0234 0 1>}
    B=ScalarParameter() {defaultValue=50, group=RVT, parameterName=Blend RVT, sortPriority=0}
   ))
  n25 = TextureSample(UVs=Divide(A=@n1, B=ScalarParameter() {defaultValue=2034.6349, parameterName=SizeNoise})) {samplerType=SAMPLERTYPE_LinearColor, texture=/Game/Global/Textures/Default/Noise/T_Noise01_Inverse}
  n26 = VectorParameter() {parameterName=ColorRubble1, defaultValue=<rgba 0.2969 0.2969 0.2969 1>}
  n27 = TextureSampleParameter2D(UVs=@n10) {parameterName=RubbleDiffuse1, texture=/Game/Environments/Sets/Rubble/Props/Textures/T_rubble_pile_D}
  n28 = TextureSample(UVs=@n10) {samplerType=SAMPLERTYPE_Masks, texture=/Game/Environments/Sets/Rubble/Props/Textures/T_rubble_pile_S}
  n29 = LinearInterpolate(
   A=ScalarParameter() {defaultValue=-0.8667, parameterName=Spec_Rubble_1}
   B=ScalarParameter() {defaultValue=1.7328, parameterName=Spec_Rubble_2}
   Alpha=@n28.R
  )
  n30 = VectorParameter() {parameterName=ColorRubble2, defaultValue=<rgba 0.1042 0.1042 0.1042 1>}
  n31 = TextureSampleParameter2D(UVs=@n6) {parameterName=RubbleDiffuse2, texture=&T_rubble_pile_D}
  n32 = TextureSample(UVs=@n6) {samplerType=SAMPLERTYPE_Masks, texture=&T_rubble_pile_S}
  n33 = LinearInterpolate(
   A=ScalarParameter() {defaultValue=-0.8667, parameterName=Spec_Rubble_1}
   B=ScalarParameter() {defaultValue=1.7328, parameterName=Spec_Rubble_2}
   Alpha=@n32.R
  )
  n34 = NamedRerouteDeclaration(None=@n18.R) {name=HM, nodeColor=<rgba 0 0.9062 1 1>}
  n35 = MaterialFunctionCall!(
   Base Material=MakeMaterialAttributes!(
    BaseColor=Multiply(A=@n26.RGB, B=@n27.RGB)
    Metallic=@n28.G
    Specular=Saturate(None=@n29)
    Roughness=@n28.G
    Anisotropy
    EmissiveColor
    Opacity=TextureSample(UVs=@n10) {samplerType=SAMPLERTYPE_Masks, texture=&T_Rubble_Pile_H}.R
    OpacityMask
    Normal=@n11.RGB
    Tangent
    WorldPositionOffset
    SubsurfaceColor
    ClearCoat
    ClearCoatRoughness
    AmbientOcclusion
    Refraction
    CustomizedUVs_0
    CustomizedUVs_1
    CustomizedUVs_2
    CustomizedUVs_3
    CustomizedUVs_4
    CustomizedUVs_5
    CustomizedUVs_6
    CustomizedUVs_7
    PixelDepthOffset
    ShadingModel
    Displacement
   )
   Top Material=MakeMaterialAttributes(BaseColor=Multiply(A=@n30.RGB, B=@n31.RGB), Metallic=@n32.G, Specular=Saturate(None=@n33), Roughness=@n32.G, Opacity=@n34, Normal=@n7.RGB)
   Alpha=MaterialFunctionCall(
    Transition Phase=ScalarParameter() {defaultValue=0.6095, parameterName=Transition}
    Height Texture=@n34
    Contrast=ScalarParameter() {defaultValue=1.5017, parameterName=TransitionContrast}
   ) {materialFunction=&HeightLerp}.Alpha
  ) {materialFunction=/Engine/Functions/MaterialLayerFunctions/MatLayerBlend_Standard}
  n36 = VertexColor()
  n37 = LinearInterpolate(
   A=VectorParameter() {parameterName=ColorSoil, defaultValue=<rgba 0.18 0.1442 0.1425 1>}.RGB
   B=VectorParameter() {parameterName=ColorSoilNeutral, defaultValue=<rgba 0.5 0.5 0.5 1>}.RGB
   Alpha=@n36.B
  )
  n38 = TextureSample(UVs=@n2) {texture=/FigmentEnvironment/LM_8_1_1/Packages/DS_Fortnight/WhiteRockCustom_D}
  n39 = LinearInterpolate(A=ScalarParameter() {parameterName=S_Soil}, B=ScalarParameter() {defaultValue=1.2085, parameterName=S_Soil2}, Alpha=@n38.R)
  n40 = ScalarParameter() {defaultValue=0.1368, parameterName=S_SoilBlendtoOldMat}
  n41 = ScalarParameter() {defaultValue=0.581, parameterName=RSoil}
  n42 = ScalarParameter() {defaultValue=1, parameterName=RSoil2}
  n43 = Multiply(A=@n3.RGB, B=VectorParameter() {parameterName=Normal, defaultValue=<rgba 1 1 1 1>}.RGB)
  n44 = ScalarParameter() {parameterName=SwizzleGravelNormal}
  n45 = MaterialFunctionCall(
   Transition Phase=Multiply(A=ScalarParameter() {defaultValue=0.5861, parameterName=Transition2}, B=@n25.R)
   Height Texture=OneMinus(None=GetMaterialAttributes(None=@n35.Blended Material) {attributeGetTypes [#5]}.Opacity)
   Contrast=ScalarParameter() {defaultValue=1.5017, parameterName=TransitionContrast2}
  ) {materialFunction=&HeightLerp}
  n46 = VertexColor()
  n47 = MaterialFunctionCall(
   Base Material=@n35.Blended Material
   Top Material=MakeMaterialAttributes(
    BaseColor=Multiply(A=@n37, B=@n38.RGB)
    Metallic=Constant()
    Specular=LinearInterpolate(A=@n39, B=@n40, Alpha=@n36.B)
    Roughness=LinearInterpolate(A=@n41, B=@n42, Alpha=@n36.B)
    Normal=LinearInterpolate(A=@n3.RGB, B=@n43, Alpha=@n44)
   )
   Alpha=Saturate(None=Add(A=@n45.Alpha, B=@n46.B))
  ) {materialFunction=&MatLayerBlend_Standard}
  n48 = GetMaterialAttributes(None=StaticSwitchParameter(
    True=MaterialFunctionCall!(Noise=@n25.R, In=@n47.Blended Material) {materialFunction=/Game/Athena/Environments/Landscape/MaterialFunctions/MF_DamageBurn}.Result
    False=@n47.Blended Material
   ) {defaultValue=true, parameterName=UseDamageBurn}) {attributeGetTypes [#4]}
  n49 = SetMaterialAttributes!(MaterialAttributes=@n48.MaterialAttributes, Normal=Transform(None=@n48.Normal)) {attributeSetTypes [#4]}
  n50 = MaterialFunctionCall(
   NoRVT=@n49
   RVT=MaterialProxyReplace(Realtime=BlendMaterialAttributes(A=@n20, B=@n49, Alpha=@n24) {vertexAttributeBlendType=UseA}, MaterialProxy=@n49)
  ) {materialFunction=&FN_RVT_Switch_Material}
  StaticSwitchParameter(True=@n49, False=ShadingPathSwitch(Default=@n50.Result, Deferred=@n50.Result, Forward=@n49, Mobile=@n49)) {group=RVT, parameterName=Skip RVT, sortPriority=-4}
 ]
}
