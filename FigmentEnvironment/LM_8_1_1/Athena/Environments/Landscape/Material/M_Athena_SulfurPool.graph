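# Landscape material for sulfur pools, usable on spline meshes, shading from colored water through wet orange sand and mud to a gravel rim.
# UV gradients, perturbed by world-aligned noise, drive HeightLerp bands using sand and lava textures.
# Layered panning normals form waves, flattened with camera distance, that also distort submerged lava-mask lookups.
# An IsWaterOnly static switch blends in plain water through material-attribute layering.
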Material /FigmentEnvironment/LM_8_1_1/Athena/Environments/Landscape/Material/M_Athena_SulfurPool
graph {
 properties {
  bTangentSpaceNormal=false
  bUseMaterialAttributes=true
  bUsedWithSplineMeshes=true
  ditherOpacityMask=true
  pixelDepthOffsetMode=PDOM_Legacy
  usedShadingModels=Default Lit
 }
 nodes [
  Constant3Vector() {constant=<rgba 0.0364 0.1 0.0467 1>}
  Constant3Vector() {constant=<rgba 0.0081 0.1 0.0642 1>}
  Constant3Vector() {constant=<rgba 0.15 0.0888 0.0075 0>}
  Constant() {r=400}
  Constant() {r=-0.3}
  Constant() {r=4}
  Constant() {r=0.6}
  Constant() {r=-3.5}
  Constant() {r=8}
  Constant() {r=-1}
  Constant() {r=3.5}
  Constant() {r=-1.6}
  Constant() {r=2.7}
  Constant() {r=-3}
  Constant() {r=4.5}
  Constant() {r=800}
  Constant() {r=-2.5}
  Constant() {r=5.5}
  n1 = ScalarParameter() {parameterName=WaterRoughness}
  n2 = ScalarParameter() {defaultValue=0.4, parameterName=WetSandRoughness}
  n3 = ScalarParameter() {defaultValue=1, parameterName=Sand Roughness Max}
  n4 = Multiply(A=ComponentMask(None=TextureCoordinate()) {g=true}, B=ScalarParameter() {defaultValue=1.15, parameterName=One-X})
  n5 = ScalarParameter() {defaultValue=-0.3, parameterName=One-Add}
  n6 = TextureObjectParameter() {
   group=Mud
   parameterName=Sand_Diffuse
   samplerSource=SSM_Wrap_WorldGroupSettings
   texture=/FigmentEnvironment/LM_8_1_1/Athena/Environments/Landscape/Textures/T_Athena_BeachSand_01
  }
  n7 = MaterialFunctionCall(
   TextureObject=@n6
   TextureSize=ScalarParameter() {defaultValue=800, group=Mud, parameterName=MudTextureTilingFrequency}
   Export Float 4=StaticBool() {value=true}
  ) {materialFunction=&WorldAlignedTexture}
  n8 = ComponentMask(None=@n7.Z Texture) {a=true}
  n9 = Add(
   A=Multiply(A=@n8, B=ScalarParameter() {defaultValue=3.5, parameterName=HL-zz4})
   B=ScalarParameter() {defaultValue=-0.9, parameterName=HL-zz3}
  )
  n10 = MaterialFunctionCall(
   Transition Phase=Saturate(None=Add(A=@n4, B=@n5))
   Height Texture=@n8
   Contrast=LinearInterpolate(A=ScalarParameter() {defaultValue=15, parameterName=OrangeSandBlendSharpness}, B=Constant(), Alpha=Saturate(None=@n9))
  ) {materialFunction=&HeightLerp}
  n11 = TextureObject() {samplerType=SAMPLERTYPE_Masks, texture=/FigmentEnvironment/LM_8_1_1/Environments/Sets/Vulcano/Textures/Lava_M}
  n12 = WorldPosition() {worldPositionShaderOffset=WPT_ExcludeAllShaderOffsets}
  n13 = ComponentMask(None=TextureCoordinate()) {g=true}
  n14 = Multiply(A=@n13, B=ScalarParameter() {defaultValue=8, parameterName=hi76002})
  n15 = ScalarParameter() {defaultValue=-3.25, parameterName=ku8i78676002}
  n16 = ScalarParameter() {defaultValue=2, parameterName=NormUVDist}
  n17 = ComponentMask(None=WorldPosition().XYZ) {g=true, r=true}
  n18 = ScalarParameter() {defaultValue=1000, parameterName=44sdsdf3}
  n19 = ScalarParameter() {defaultValue=500, group=Mud, parameterName=ClumpySize}
  n20 = ComponentMask(None=MaterialFunctionCall(TextureObject=@n11, TextureSize=@n19) {materialFunction=&WorldAlignedTexture}.Z Texture) {g=true}
  n21 = Add(A=Add(A=Multiply(A=@n8, B=@n16), B=Divide(A=@n17, B=@n18)), B=Multiply(A=ScalarParameter() {defaultValue=1, parameterName=boof}, B=@n20))
  n22 = OneMinus(None=Power(
    Base=TextureSample(UVs=@n21) {texture=/FigmentEnvironment/LM_8_1_1/Athena/Items/Consumables/Spooky_Mist_Cube/Textures/T_VertGradient}.R
    Exp=ScalarParameter() {defaultValue=0.5, parameterName=nnnf}
   ))
  n23 = MaterialFunctionCall(
   Transition Phase=Saturate(None=Add(A=@n14, B=@n15))
   Height Texture=@n22
   Contrast=ScalarParameter() {defaultValue=0.5, parameterName=contage3}
  ) {materialFunction=&HeightLerp}
  n24 = Multiply(A=OneMinus(None=@n23.Alpha), B=ScalarParameter() {defaultValue=1.5, parameterName=gogogoj})
  n25 = ScalarParameter() {defaultValue=-1, parameterName=wdcsfgdgf}
  n26 = Divide(A=ComponentMask(None=WorldPosition().XYZ) {g=true, r=true}, B=ScalarParameter() {defaultValue=603.1323, parameterName=WaveSize})
  n27 = Add(A=Multiply(A=@n26) {constB=0.8556}) {constB=0.3215}
  n28 = Add(A=@n26) {constB=0.1255}
  n29 = ComponentMask(None=TextureSample(UVs=Panner(Coordinate=@n28) {bFractionalPart=true, speedX=0.0208, speedY=-0.0278}) {samplerType=SAMPLERTYPE_Normal, texture=/FigmentEnvironment/LM_8_1_1/Packages/Fortnite_UniqueMaterials_1/Textures/T_Water2_N}.RGB) {g=true, r=true}
  n30 = TextureSample(UVs=Add(A=Panner(Coordinate=@n27) {bFractionalPart=true, speedX=0.0625, speedY=0.0417}, B=Multiply(A=@n29) {constB=0.035})) {samplerType=SAMPLERTYPE_Normal, texture=&T_Water2_N}
  n31 = Add(
   A=Multiply(A=ComponentMask(None=@n30.RGB) {g=true, r=true}, B=ScalarParameter() {defaultValue=0.0442, group=Deep Water, parameterName=WaveNoise}) {constB=0.04}
   B=Add(A=Panner(Coordinate=@n26) {bFractionalPart=true, speedX=-0.025, speedY=-0.0375}) {constB=0.333}
  )
  n32 = Normalize(VectorInput=MaterialFunctionCall(
    BaseNormal=@n30.RGB
    AdditionalNormal=TextureSample(UVs=@n31) {samplerType=SAMPLERTYPE_Normal, texture=&T_Water2_N}.RGB
   ) {materialFunction=&BlendAngleCorrectedNormals}.Result)
  n33 = Multiply(A=Saturate(None=Add(A=@n24, B=@n25)), B=Multiply(A=@n32, B=ScalarParameter() {defaultValue=20, parameterName=rthbfy45y}))
  n34 = ComponentMask(None=MaterialFunctionCall(TextureObject=@n11, TextureSize=@n19, WorldPosition=Add(A=@n12.XYZ, B=@n33)) {materialFunction=&WorldAlignedTexture}.Z Texture) {g=true}
  n35 = Add(
   A=Multiply(A=@n34, B=ScalarParameter() {defaultValue=2, parameterName=mkjhuiy})
   B=ScalarParameter() {defaultValue=-0.5, parameterName=4287g}
  )
  n36 = Saturate(None=Add(A=LinearInterpolate(A=@n2, B=@n3, Alpha=@n10.Alpha), B=Saturate(None=@n35)))
  n37 = TextureObject() {samplerType=SAMPLERTYPE_Normal, texture=/FigmentEnvironment/LM_8_1_1/Environments/Foliage/Textures/T_SwirlNoise_N}
  n38 = ScalarParameter() {defaultValue=2500, parameterName=PerperbScale}
  n39 = StaticBool() {value=true}
  n40 = ComponentMask(None=Multiply(
    A=ScalarParameter() {defaultValue=0.4, parameterName=PerperbAmount}
    B=MaterialFunctionCall(TextureObject=@n37, TextureSize=@n38, WorldSpace=@n39) {materialFunction=&WorldAlignedNormal}.Z Texture
   )) {g=true, r=true}
  n41 = ComponentMask(None=Add(A=@n40, B=TextureCoordinate())) {g=true}
  n42 = ScalarParameter() {defaultValue=2, parameterName=gh54y3}
  n43 = Saturate(None=Add(A=Multiply(A=@n41, B=@n42), B=ScalarParameter() {parameterName=ngfjy666}))
  LinearInterpolate(A=@n1, B=LinearInterpolate(A=ScalarParameter() {parameterName=VeryWetSandRoughness}, B=@n36), Alpha=@n43)
  n44 = WorldPosition()
  n45 = TextureSample(UVs=Divide(A=ComponentMask(None=@n44.XYZ) {g=true, r=true}, B=Constant() {r=-700})) {
   samplerSource=SSM_Wrap_WorldGroupSettings
   samplerType=SAMPLERTYPE_LinearColor
   texture=/FigmentEnvironment/LM_8_1_1/Athena/Environments/Landscape/Textures/Athena_GrassBiome_M_02
  }
  n46 = LinearInterpolate(
   A=Multiply(A=Constant3Vector() {constant=<rgba 0.1979 0.1709 0.1554 0>}, B=@n45.B)
   B=Multiply(A=Constant3Vector() {constant=<rgba 0.244 0.2799 0.2917 0>}, B=@n45.B)
   Alpha=MaterialFunctionCall(In=@n45.B, Contrast=Constant() {r=2}) {materialFunction=&CheapContrast}.Result
  )
  Multiply(
   A=@n46
   B=ComponentMask(None=CollectionParameter() {collection=/FigmentEnvironment/LM_8_1_1/Athena/Environments/Landscape/MPC/MPC_Landscape, parameterName=JunglePools_GravelTint_01}) {b=true, g=true, r=true}
  )
  n47 = VectorParameter() {group=Mud, parameterName=SandTintSecondColorWET, defaultValue=<rgba 0.15 0.0888 0.0075 0>}
  n48 = ComponentMask(None=@n7.Z Texture) {b=true, g=true, r=true}
  Multiply(A=@n47.RGB, B=@n48)
  n49 = Multiply(A=VectorParameter() {parameterName=GravelPoolEdgeTint, defaultValue=<rgba 1.5 1.1733 0.8777 0>}.RGB, B=@n46)
  n50 = TextureCoordinate()
  n51 = Add(
   A=Multiply(A=ComponentMask(None=@n50) {g=true}, B=ScalarParameter() {defaultValue=1.5, parameterName=outerblendX04})
   B=ScalarParameter() {defaultValue=-0.5, parameterName=outerblendX05}
  )
  n52 = Add(
   A=Multiply(A=@n45.G, B=ScalarParameter() {defaultValue=4, parameterName=GravMudH-X})
   B=LinearInterpolate(
    A=ScalarParameter() {defaultValue=-5, parameterName=GravMudH-+}
    B=ScalarParameter() {defaultValue=-0.5, parameterName=GravMudH-+high}
    Alpha=Saturate(None=@n51)
   )
  )
  n53 = Multiply(A=ComponentMask(None=TextureCoordinate()) {g=true}, B=ScalarParameter() {defaultValue=1.5, parameterName=HL-03new})
  n54 = ScalarParameter() {defaultValue=-0.5, parameterName=HL-04new}
  n55 = Multiply(A=OneMinus(None=@n20), B=ScalarParameter() {defaultValue=4.5, parameterName=HL-03new2})
  n56 = ScalarParameter() {defaultValue=-2.8, parameterName=HL-04new2}
  n57 = MaterialFunctionCall(Transition Phase=Saturate(None=Add(A=@n53, B=@n54)), Height Texture=Saturate(None=Add(A=@n55, B=@n56))) {materialFunction=&HeightLerp}
  n58 = LinearInterpolate(
   A=VectorParameter() {parameterName=GravelPoolMud, defaultValue=<rgba 0.12 0.0641 0.036 0>}.RGB
   B=@n49
   Alpha=Multiply(A=Saturate(None=@n52), B=@n57.Lerp Alpha No Contrast)
  )
  n59 = Multiply(A=@n46, B=Constant3Vector() {constant=<rgba 0.9844 0.7659 0.6562 0>})
  n60 = LinearInterpolate(
   A=Multiply(A=VectorParameter() {group=Mud, parameterName=SandTintDark, defaultValue=<rgba 0.2 0.1034 0.04 0>}.RGB, B=@n48)
   B=FeatureLevelSwitch(Default=@n58, ES2=@n59, ES3_1=@n59, SM4_REMOVED=@n58, SM5=@n58, SM6=@n58)
   Alpha=@n57.Lerp Alpha No Contrast
  )
  n61 = VectorParameter() {group=Mud, parameterName=SandTintFourthColor, defaultValue=<rgba 0.0016 0.03 0.0297 1>}
  n62 = VectorParameter() {group=Mud, parameterName=SandTintSecondColor, defaultValue=<rgba 0.9 0.5314 0.0367 0>}
  n63 = VectorParameter() {group=Mud, parameterName=SandTintFirstColor, defaultValue=<rgba 0.55 0.2325 0.0241 0>}
  n64 = ScalarParameter() {defaultValue=5.5, parameterName=SecondColor-X}
  n65 = Saturate(None=Add(A=Multiply(A=@n41, B=@n64), B=ScalarParameter() {defaultValue=-2.5, parameterName=SecondColor-+}))
  n66 = TextureCoordinate()
  n67 = Add(
   A=Multiply(A=ComponentMask(None=@n66) {g=true}, B=ScalarParameter() {defaultValue=8, parameterName=WetSand-X})
   B=ScalarParameter() {defaultValue=-3.25, parameterName=WetSand-+}
  )
  n68 = LinearInterpolate(
   A=@n47.RGB
   B=LinearInterpolate(
    A=Multiply(A=@n62.RGB, B=@n48)
    B=Multiply(A=@n63.RGB, B=@n48)
    Alpha=MaterialFunctionCall(Transition Phase=@n65, Height Texture=@n22) {materialFunction=&HeightLerp}.Lerp Alpha No Contrast
   )
   Alpha=MaterialFunctionCall(Transition Phase=Saturate(None=@n67), Height Texture=@n22) {materialFunction=&HeightLerp}.Lerp Alpha No Contrast
  )
  n69 = Saturate(None=Add(A=ScalarParameter() {defaultValue=0.5, parameterName=damnvariablenames}, B=@n34))
  n70 = Power(Base=@n34, Exp=ScalarParameter() {defaultValue=2, parameterName=holepower})
  n71 = MaterialFunctionCall(Transition Phase=@n43, Height Texture=@n70) {materialFunction=&HeightLerp}
  n72 = ComponentMask(None=Add(A=@n40, B=TextureCoordinate() {coordinateIndex=1})) {g=true}
  n73 = ScalarParameter() {defaultValue=1, parameterName=fdfdffdr}
  n74 = Saturate(None=Add(A=Multiply(A=@n72, B=@n73), B=ScalarParameter() {defaultValue=0.3, parameterName=mmmmkkju}))
  n75 = LinearInterpolate(
   A=@n61.RGB
   B=LinearInterpolate(
    A=VectorParameter() {group=Mud, parameterName=SandTintThirdColor, defaultValue=<rgba 0.0225 0.1 0.0603 1>}.RGB
    B=Multiply(A=@n68, B=@n69)
    Alpha=Multiply(A=@n43, B=@n71.Lerp Alpha No Contrast)
   )
   Alpha=Multiply(
    A=@n74
    B=MaterialFunctionCall(Transition Phase=@n74, Height Texture=@n70) {materialFunction=&HeightLerp}.Lerp Alpha No Contrast
   )
  )
  n76 = ScalarParameter() {defaultValue=0.4, parameterName=WaterSpec}
  n77 = ScalarParameter() {defaultValue=0.1, parameterName=Sand Spec}
  n78 = Multiply(A=@n36, B=@n23.Alpha)
  n79 = VertexColor()
  n80 = ComponentMask(None=@n32) {r=true}
  n81 = Constant() {r=6000}
  n82 = Constant() {r=2000}
  n83 = MaterialFunctionCall(
   Normal=@n32
   Flatness=LinearInterpolate(
    A=ScalarParameter() {defaultValue=0.89, parameterName=WaveNormalFlatten}
    B=ScalarParameter() {defaultValue=0.99, parameterName=WaveNormalFlatten_2}
    Alpha=MaterialFunctionCall(Fade Length=@n81, Fade Offset=@n82) {materialFunction=&CameraDepthFade}.Result
   )
  ) {materialFunction=&FlattenNormal}
  n84 = Add(
   A=Multiply(A=OneMinus(None=@n13), B=ScalarParameter() {defaultValue=40, parameterName=NROR-X})
   B=ScalarParameter() {defaultValue=-1, parameterName=NROR-+}
  )
  n85 = TextureObject() {samplerType=SAMPLERTYPE_Normal, texture=/FigmentEnvironment/LM_8_1_1/Environments/Sets/Vulcano/Textures/Lava_N}
  n86 = StaticBool() {value=true}
  n87 = Add(
   A=Multiply(A=@n8, B=ScalarParameter() {defaultValue=4, parameterName=vcszf3})
   B=ScalarParameter() {defaultValue=-1.5, parameterName=mmmn664}
  )
  n88 = LinearInterpolate(
   A=Constant3Vector() {constant=<rgba 0 0 1 0>}
   B=MaterialFunctionCall(
    A=MaterialFunctionCall(TextureObject=@n85, TextureSize=@n19, WorldSpace=@n86) {materialFunction=&WorldAlignedNormal}.Z Texture
    B=VertexNormalWS()
    Transition Phase=Saturate(None=@n87)
    Height Texture=@n34
   ) {materialFunction=&HeightLerp}.Results
   Alpha=Multiply(A=Saturate(None=@n84), B=@n23.Alpha)
  )
  n89 = MaterialFunctionCall!(
   Base Material=MakeMaterialAttributes!(
    BaseColor=LinearInterpolate(A=@n75, B=@n60, Alpha=@n10.Alpha)
    Metallic
    Specular=LinearInterpolate(A=@n76, B=@n77, Alpha=@n43)
    Roughness=Multiply(A=@n78, B=@n43)
    Anisotropy
    EmissiveColor
    Opacity
    OpacityMask=MaterialFunctionCall(Transition Phase=@n79.B, Height Texture=@n80) {materialFunction=&HeightLerp}.Lerp Alpha No Contrast
    Normal=LinearInterpolate(A=@n83.Result, B=@n88, Alpha=@n43)
    Tangent
    WorldPositionOffset
    SubsurfaceColor
    ClearCoat
    ClearCoatRoughness
    AmbientOcclusion
    Refraction
    CustomizedUVs_0
    CustomizedUVs_1
    CustomizedUVs_2
    CustomizedUVs_3
    CustomizedUVs_4
    CustomizedUVs_5
    CustomizedUVs_6
    CustomizedUVs_7
    PixelDepthOffset
    ShadingModel
    Displacement
   )
   Top Material=MakeMaterialAttributes(BaseColor=@n49, Specular=ScalarParameter() {defaultValue=0.1, parameterName=Gravel-Spec}, Roughness=Constant() {r=1})
   Alpha=Constant()
  ) {materialFunction=/Engine/Functions/MaterialLayerFunctions/MatLayerBlend_Standard}
  StaticSwitchParameter(
   True=MaterialFunctionCall(
    Base Material=@n89.Blended Material
    Top Material=MakeMaterialAttributes(BaseColor=@n61.RGB, Specular=@n76, Roughness=@n1, Normal=@n83.Result)
    Alpha=ScalarParameter() {defaultValue=1, parameterName=WaterOnly}
   ) {materialFunction=&MatLayerBlend_Standard}.Blended Material
   False=@n89.Blended Material
  ) {parameterName=IsWaterOnly}
 ]
}
