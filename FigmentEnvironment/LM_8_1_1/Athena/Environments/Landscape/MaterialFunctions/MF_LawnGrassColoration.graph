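# Material function that recolors landscape lawn grass for golf fairways and greens, with Material_Low and Material_Standard outputs.
# Base color's red channel drives a green tint, blended by the fairway mask, with rotated, rounded landscape coordinates adding two-tone mowing stripes in its core.
# Standard height-blends using inverted opacity, then uses the Golf_Green layer to brighten color and flatten normals.
# Low blends directly and replaces stripes with flat brightness on mobile feature levels.
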MaterialFunction /FigmentEnvironment/LM_8_1_1/Athena/Environments/Landscape/MaterialFunctions/MF_LawnGrassColoration
graph {nodes [
  n1 = FunctionInput() {bUsePreviewValueAsDefault=true, inputName=Material_Low, inputType=FunctionInput_MaterialAttributes, sortPriority=1}
  n2 = Constant() {r=0.075}
  n3 = Constant() {r=1}
  n4 = MaterialFunctionCall!(Material=@n1) {materialFunction=/Engine/Functions/MaterialLayerFunctions/MatLayerBlend_BreakBaseColor}
  n5 = Constant3Vector() {constant=<rgba 0.5 1.05 0.2516 1>}
  n6 = Multiply(A=LinearInterpolate(A=@n2, B=@n3, Alpha=ComponentMask(None=@n4.BaseColor) {r=true}), B=@n5)
  n7 = FunctionInput() {bUsePreviewValueAsDefault=true, inputName=Golf_Fairway_Mask, inputType=FunctionInput_Scalar, sortPriority=2}
  n8 = Constant() {r=1.3}
  n9 = LandscapeLayerCoords() {mappingScale=4}
  n10 = Constant() {r=0.4}
  n11 = Frac(None=ComponentMask(None=MaterialFunctionCall(UVs=@n9, "Rotation Angle (0-1)"=@n10) {materialFunction=&CustomRotator}.Rotated Values) {r=true})
  n12 = LinearInterpolate(A=Constant() {r=1.1}, B=@n8, Alpha=Clamp(None=Round(None=@n11)))
  n13 = FeatureLevelSwitch(Default=@n12, ES2=@n8, ES3_1=@n8, SM4_REMOVED=@n12, SM5=@n12, SM6=@n12)
  n14 = Clamp(None=LinearInterpolate(A=Constant() {r=-4}, B=Constant() {r=1}, Alpha=@n7))
  FunctionOutput(None=MaterialFunctionCall!(
    Material=MaterialFunctionCall(Material=@n1, NewBaseColor=@n6, Mask=@n7) {materialFunction=/Engine/Functions/MaterialLayerFunctions/MatLayerBlend_OverrideBaseColor}.Blended Material
    NewBaseColor=Multiply(A=@n6, B=@n13)
    Mask=@n14
   ) {materialFunction=&MatLayerBlend_OverrideBaseColor}.Blended Material) {outputName=Material_Low, sortPriority=1}
  n15 = FunctionInput() {bUsePreviewValueAsDefault=true, inputType=FunctionInput_MaterialAttributes}
  n16 = MaterialFunctionCall(Material=@n15) {materialFunction=&MatLayerBlend_BreakBaseColor}
  n17 = Multiply(A=LinearInterpolate(A=@n2, B=@n3, Alpha=ComponentMask(None=@n16.BaseColor) {r=true}), B=@n5)
  n18 = MaterialFunctionCall(
   Transition Phase=@n7
   Height Texture=OneMinus(None=MaterialFunctionCall!(Mat attributes=@n15) {materialFunction=/Engine/Functions/Engine_MaterialFunctions02/MatLayerBlend_BreakOpacity}.Result)
   Contrast=Constant() {r=1}
  ) {materialFunction=&HeightLerp}
  n19 = MaterialFunctionCall(
   Material=MaterialFunctionCall(Material=@n15, NewBaseColor=@n17, Mask=@n18.Alpha) {materialFunction=&MatLayerBlend_OverrideBaseColor}.Blended Material
   NewBaseColor=Multiply(A=@n17, B=@n12)
   Mask=@n14
  ) {materialFunction=&MatLayerBlend_OverrideBaseColor}
  n20 = Multiply(A=@n17, B=Constant() {r=1.3})
  n21 = LandscapeLayerSample() {parameterName=Golf_Green}
  FunctionOutput(None=MaterialFunctionCall!(
    Material=MaterialFunctionCall(Material=@n19.Blended Material, NewBaseColor=@n20, Mask=@n21) {materialFunction=&MatLayerBlend_OverrideBaseColor}.Blended Material
    NormalFlatness=@n21
    Normal
   ) {materialFunction=/Engine/Functions/MaterialLayerFunctions/MatLayerBlend_NormalFlatten}.Blended Material) {outputName=Material_Standard}
  Multiply(A=Power(Base=@n18.Alpha, Exp=Constant() {r=0.5}), B=@n17)
  Saturate(None=Add(A=@n7, B=@n21)) {desc=Issue?}
 ]}
