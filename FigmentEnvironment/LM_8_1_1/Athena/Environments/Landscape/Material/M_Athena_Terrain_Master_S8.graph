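# Master landscape material for Athena terrain, blending grass, arid, ice and snow biomes from material functions with slope-based rock transitions.
# World-space colormaps and MPC_Landscape tints color the biomes, then forest, mud, road and gravel layers height-blend over them.
# Quality and feature-level switches provide cheaper paths; EnableQuickColorRender adds colormap and biome-mask render modes.
# Landscape grass outputs all take one zero constant labeled NoGrassPCG, and a runtime virtual texture is sampled.
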Material /FigmentEnvironment/LM_8_1_1/Athena/Environments/Landscape/Material/M_Athena_Terrain_Master_S8
graph {
 properties {bUseMaterialAttributes=true, pixelDepthOffsetMode=PDOM_Legacy, refractionMethod=RM_IndexOfRefraction, usedShadingModels=Default Lit}
 nodes [
  Constant() {r=0.2}
  n1 = ScalarParameter() {defaultValue=1500, group=MasterControl, parameterName=Texturesize_Z_Sand}
  Divide(
   A=ComponentMask(None=WorldPosition() {worldPositionShaderOffset=WPT_ExcludeAllShaderOffsets}.XYZ) {g=true, r=true}
   B=Multiply(A=Abs(None=@n1)) {constB=-1}
  )
  MaterialExpressionConstant_1 = Constant() {desc=NoGrassPCG}
  LandscapeGrassOutput!(
   Grass=@MaterialExpressionConstant_1
   Mountain=@MaterialExpressionConstant_1
   Farm=@MaterialExpressionConstant_1
   Forest=@MaterialExpressionConstant_1
   Gravel=@MaterialExpressionConstant_1
   Flowers=@MaterialExpressionConstant_1
   Bushes=@MaterialExpressionConstant_1
   Crater=@MaterialExpressionConstant_1
   Arid_Sand=@MaterialExpressionConstant_1
   Snow=@MaterialExpressionConstant_1
  ) {grassTypes [
    Grass {
     grassType=/FigmentEnvironment/LM_8_1_1/Athena/Environments/Landscape/Foliage/Athena_Foliage_Grass
     input {
      expression=&self:MaterialExpressionConstant_1
      inputName=None
      mask=0
      maskA=0
      maskB=0
      maskG=0
      maskR=0
      outputIndex=0
     }
    }
    Mountain {
     grassType=/FigmentEnvironment/LM_8_1_1/Athena/Environments/Landscape/Foliage/Athena_Foliage_Mountain_Grass
     input {
      expression=&self:MaterialExpressionConstant_1
      inputName=None
      mask=0
      maskA=0
      maskB=0
      maskG=0
      maskR=0
      outputIndex=0
     }
    }
    Farm {
     grassType=/FigmentEnvironment/LM_8_1_1/Athena/Environments/Landscape/Foliage/Athena_Foliage_CraterOvergrowth
     input {
      expression=&self:MaterialExpressionConstant_1
      inputName=None
      mask=0
      maskA=0
      maskB=0
      maskG=0
      maskR=0
      outputIndex=0
     }
    }
    Forest {
     grassType=/FigmentEnvironment/LM_8_1_1/Athena/Environments/Landscape/Foliage/Athena_Foliage_Forest
     input {
      expression=&self:MaterialExpressionConstant_1
      inputName=None
      mask=0
      maskA=0
      maskB=0
      maskG=0
      maskR=0
      outputIndex=0
     }
    }
    Gravel {
     grassType=/FigmentEnvironment/LM_8_1_1/Athena/Environments/Landscape/Foliage/Athena_Foliage_Gravel
     input {
      expression=&self:MaterialExpressionConstant_1
      inputName=None
      mask=0
      maskA=0
      maskB=0
      maskG=0
      maskR=0
      outputIndex=0
     }
    }
    Flowers {
     grassType=/FigmentEnvironment/LM_8_1_1/Athena/Environments/Landscape/Foliage/Athena_Foliage_Flowers
     input {
      expression=&self:MaterialExpressionConstant_1
      inputName=None
      mask=0
      maskA=0
      maskB=0
      maskG=0
      maskR=0
      outputIndex=0
     }
    }
    Bushes {
     grassType=/FigmentEnvironment/LM_8_1_1/Athena/Environments/Landscape/Foliage/Athena_Foliage_Bushes
     input {
      expression=&self:MaterialExpressionConstant_1
      inputName=None
      mask=0
      maskA=0
      maskB=0
      maskG=0
      maskR=0
      outputIndex=0
     }
    }
    Crater {
     grassType=/FigmentEnvironment/LM_8_1_1/Athena/Environments/Landscape/Foliage/Athena_Foliage_Crater
     input {
      expression=&self:MaterialExpressionConstant_1
      inputName=None
      mask=0
      maskA=0
      maskB=0
      maskG=0
      maskR=0
      outputIndex=0
     }
    }
    Arid_Sand {
     grassType=/FigmentEnvironment/LM_8_1_1/Athena/Environments/Landscape/Foliage/Athena_Foliage_Arid_Grass
     input {
      expression=&self:MaterialExpressionConstant_1
      inputName=None
      mask=0
      maskA=0
      maskB=0
      maskG=0
      maskR=0
      outputIndex=0
     }
    }
    Snow {
     grassType=/FigmentEnvironment/LM_8_1_1/Athena/Environments/Landscape/Foliage/Athena_Foliage_Snow
     input {
      expression=&self:MaterialExpressionConstant_1
      inputName=None
      mask=0
      maskA=0
      maskB=0
      maskG=0
      maskR=0
      outputIndex=0
     }
    }
   ]}
  LandscapeLayerSample() {parameterName=Flowers, previewWeight=1}
  LandscapeLayerSample() {parameterName=Arid_Grass_Light}
  n2 = Constant3Vector() {constant=<rgba 0.1354 0.0916 0.0585 0>}
  n3 = Constant3Vector() {constant=<rgba 2 1.472 1.0729 1>}
  n4 = LandscapeLayerSwitch!(
   LayerUsed=LinearInterpolate(A=@n2, B=Multiply(A=@n2, B=@n3) {constB=2.5}, Alpha=LandscapeLayerSample() {parameterName=Arid})
   LayerNotUsed=@n2
  ) {parameterName=Arid}
  MaterialExpressionMaterialFunctionCall_22 = MaterialFunctionCall() {materialFunction=/FigmentEnvironment/LM_8_1_1/Athena/Environments/Landscape/MaterialFunctions/Standard/MF_Athena_Grass_01}
  LinearInterpolate(A=@n4, B=@MaterialExpressionMaterialFunctionCall_22.BaseColor_Plain)
  LinearInterpolate(Alpha=Clamp()) {constB=0.2}
  MakeMaterialAttributes!(
   BaseColor
   Metallic
   Specular
   Roughness
   Anisotropy
   EmissiveColor
   Opacity
   OpacityMask
   Normal
   Tangent
   WorldPositionOffset
   SubsurfaceColor
   ClearCoat
   ClearCoatRoughness
   AmbientOcclusion
   Refraction
   CustomizedUVs_0
   CustomizedUVs_1
   CustomizedUVs_2
   CustomizedUVs_3
   CustomizedUVs_4
   CustomizedUVs_5
   CustomizedUVs_6
   CustomizedUVs_7
   PixelDepthOffset
   ShadingModel
   Displacement
  )
  n5 = TextureObject() {
   samplerType=SAMPLERTYPE_LinearColor
   texture=/FigmentEnvironment/LM_8_1_1/Athena/Environments/Landscape/Textures/Arid/T_Athena_Arid_Rock_01_M
  }
  n6 = MaterialFunctionCall!(AridMasks=@n5, ProjContrast=Constant() {r=10}) {materialFunction=/FigmentEnvironment/LM_8_1_1/Athena/Environments/Landscape/MaterialFunctions/Arid/MF_Athena_Arid_Rock}
  n7 = MaterialFunctionCall!(Material=@n6.Mat, NewBaseColor=@n6.Mat, Mask=LandscapeLayerSample() {parameterName=Arid_RockDetail, previewWeight=1}) {materialFunction=/Engine/Functions/MaterialLayerFunctions/MatLayerBlend_OverrideBaseColor}
  MaterialFunctionCall!(
   In=@n7.Blended Material
   Color=ComponentMask(None=CollectionParameter() {collection=/FigmentEnvironment/LM_8_1_1/Athena/Environments/Landscape/MPC/MPC_Landscape, parameterName=Arid_Rock_01_Tint}) {b=true, g=true, r=true}
  ) {materialFunction=/FigmentEnvironment/LM_8_1_1/Athena/Environments/Landscape/MaterialFunctions/MF_Athena_Material_BaseMultiply}
  MaterialFunctionCall!(Grass, Material_Standard, Material_BaseColor, Mask) {materialFunction=/FigmentEnvironment/LM_8_1_1/Athena/Environments/Landscape/MaterialFunctions/Arid/MF_Athena_AridGrass}
  n8 = LandscapeLayerSample() {parameterName=Gravel, previewWeight=1}
  MaterialFunctionCall(In=@n8, Contrast=Constant() {r=1}) {materialFunction=&CheapContrast}
  n9 = RuntimeVirtualTextureSampleParameter() {materialType=BaseColor_Normal_Specular_Mask_YCoCg, parameterName=RVT, virtualTexture=/FigmentCore/Athena/Prototype/Terrain/RVT_Figment}
  n10 = Constant()
  n11 = MaterialFunctionCall() {materialFunction=/FigmentEnvironment/LM_8_1_1/Athena/Environments/Landscape/MaterialFunctions/Standard/MF_Athena_RoadGravel}
  n12 = MaterialFunctionCall!(Gravel=AppendVector(A=@n11.Gravel, B=@n11.Gravel)) {materialFunction=/FigmentEnvironment/LM_8_1_1/Athena/Environments/Landscape/MaterialFunctions/Standard/MF_Athena_Rock_01}
  n13 = MaterialFunctionCall!(Rock=@n12.Result) {materialFunction=/FigmentEnvironment/LM_8_1_1/Athena/Environments/Landscape/MaterialFunctions/Snow/MF_Athena_IceRock_01}
  n14 = MaterialFunctionCall!(UseLandscapeCoordsOnSwitch) {materialFunction=/FigmentEnvironment/LM_8_1_1/Athena/Environments/Landscape/MaterialFunctions/Snow/MF_Athena_Snow_01}
  n15 = Constant() {r=1}
  n16 = MaterialFunctionCall!(Height=@n14.Height, Transition Multiplier=@n15, Height Multiplier) {materialFunction=/FigmentEnvironment/LM_8_1_1/Athena/Environments/Landscape/MaterialFunctions/Utilities/MF_Athena_SlopeBlend}
  n17 = Multiply(
   A=Multiply(A=OneMinus(None=@n13.EmissiveMask), B=OneMinus(None=@n16.Alpha))
   B=MaterialFunctionCall() {materialFunction=/FigmentEnvironment/LM_8_1_1/Athena/WinterArt/Materials/MF_SharedWinter}.IceColor
  )
  n18 = Multiply(
   A=Constant3Vector() {constant=<rgba 0.3346 0.4223 0.5 0>}
   B=Clamp(None=MaterialFunctionCall() {materialFunction=&MF_Athena_Snow_01}.NormalDotCamera)
  )
  n19 = LandscapeLayerSample() {parameterName=Gravel}
  n20 = LandscapeLayerSample() {parameterName=Forest}
  n21 = Subtract(A=LandscapeLayerSample() {parameterName=Ice}, B=Add(A=Add(A=@n19, B=@n20), B=LandscapeLayerSample() {parameterName=Mud})) {desc=Mask out paths and dirt from emmisive}
  n22 = StaticSwitchParameter(True=@n10, False=LinearInterpolate(A=@n10, B=Multiply(A=@n17, B=@n18) {constB=0.5}, Alpha=Clamp(None=@n21))) {group=Utilities, parameterName=EnableQuickColorRender}
  n23 = Constant()
  n24 = FeatureLevelSwitch(Default=@n22, ES2=@n23, ES3_1=@n23, SM4_REMOVED=@n22, SM5=@n22, SM6=@n22)
  n25 = SetMaterialAttributes!(
   MaterialAttributes
   Base Color=@n9.BaseColor
   Specular=@n9.BaseColor
   Roughness=@n9.BaseColor
   Opacity Mask=@n9.BaseColor
   Normal=@n9.BaseColor
   Emissive Color=@n24
  ) {attributeSetTypes [
    #1
    #2
    #3
    #4
    #5
    #6
   ]}
  n26 = LandscapeVisibilityMask()
  n27 = Constant() {r=0.1}
  n28 = Constant() {r=0.25}
  n29 = LandscapeLayerSample() {parameterName=Ice}
  n30 = MaterialFunctionCall() {materialFunction=&MF_Athena_Snow_01}
  n31 = GetMaterialAttributes(None=@n30.Material) {attributeGetTypes [#2]}
  n32 = WorldPosition()
  n33 = Constant() {r=306000}
  n34 = Subtract(A=Divide(A=ComponentMask(None=@n32.XYZ) {g=true, r=true}, B=@n33), B=Divide(A=Constant() {r=-153000}, B=@n33))
  n35 = TextureSampleParameter2D(UVs=@n34) {
   parameterName=BiomeColormap
   samplerType=SAMPLERTYPE_LinearColor
   texture=/FigmentEnvironment/LM_8_1_1/Athena/Environments/Landscape/Textures/T_Athena_Terrain_CombinedColors_01
  }
  n36 = CollectionParameter() {collection=&MPC_Landscape, parameterName=Rock_Bottom_Brightness}
  n37 = CollectionParameter() {collection=&MPC_Landscape, parameterName=Rock_Top_Brightness}
  n38 = MaterialFunctionCall!(height_Offset, height_GradScale, Coords=WorldPosition().XYZ) {materialFunction=/FigmentEnvironment/LM_8_1_1/Athena/Environments/Landscape/MaterialFunctions/MF_Athena_WorldHeightGrad}
  n39 = MaterialFunctionCall(Height=@MaterialExpressionMaterialFunctionCall_22.Height, Transition Multiplier=@n15, Height Multiplier=Constant() {r=0.1}) {materialFunction=&MF_Athena_SlopeBlend}
  MaterialExpressionMaterialFunctionCall_31 = MaterialFunctionCall!(
   Base Material=MaterialFunctionCall(In=@n12.Result, Color=LinearInterpolate(A=@n36, B=@n37, Alpha=@n38.Result)) {materialFunction=&MF_Athena_Material_BaseMultiply}.Result
   Top Material=@MaterialExpressionMaterialFunctionCall_22.Result
   Alpha=@n39.Alpha
  ) {materialFunction=/Engine/Functions/MaterialLayerFunctions/MatLayerBlend_Standard}
  n40 = MaterialFunctionCall(
   In=@n7.Blended Material
   Color=MaterialFunctionCall() {materialFunction=/FigmentEnvironment/LM_8_1_1/Athena/Environments/Landscape/MaterialFunctions/MF_Athena_SedimentGradient}.Result
  ) {materialFunction=&MF_Athena_Material_BaseMultiply}
  n41 = MaterialFunctionCall(In=@n6.MatSimple, Color=Constant3Vector() {desc="Ship hack, Replace Later", constant=<rgba 1 0.7974 0.7166 0>}) {materialFunction=&MF_Athena_Material_BaseMultiply}
  n42 = QualitySwitch(Default=@n40.Result, Low=@n41.Result, High=@n40.Result, Medium=@n41.Result)
  n43 = SetMaterialAttributes!(
   MaterialAttributes=MaterialFunctionCall!(In=@n5, UseRoughness=StaticBool()) {materialFunction=/FigmentEnvironment/LM_8_1_1/Athena/Environments/Landscape/MaterialFunctions/Arid/MF_Athena_Arid_Sand}.Simple_Material
   Normal=@n14.Normal
  ) {attributeSetTypes [#5]}
  MaterialExpressionStaticSwitchParameter_7 = StaticSwitchParameter(
   True=MaterialFunctionCall(
    Base Material=FeatureLevelSwitch(Default=@n42, ES2=@n41.Result, ES3_1=@n41.Result, SM4_REMOVED=@n42, SM5=@n42, SM6=@n42)
    Top Material=@n43
    Alpha=@n16.Alpha
   ) {materialFunction=&MatLayerBlend_Standard}.Blended Material
   False=@n43
  ) {defaultValue=true, parameterName=UseAridRockBlend}
  MaterialExpressionMaterialFunctionCall_45 = MaterialFunctionCall(Base Material=@n13.Material, Top Material=@n14.Material, Alpha=@n16.Alpha) {materialFunction=&MatLayerBlend_Standard}
  MaterialExpressionMaterialFunctionCall_11 = MaterialFunctionCall(Base Material=@n12.Result, Top Material=@n14.Material, Alpha=@n16.Alpha) {materialFunction=&MatLayerBlend_Standard}
  n44 = LandscapeLayerBlend!(
   Layer Grass_Only=@MaterialExpressionMaterialFunctionCall_22.Result
   Layer Grass_Topology=@MaterialExpressionMaterialFunctionCall_31.Blended Material
   Layer Arid=@MaterialExpressionStaticSwitchParameter_7
   Layer Ice=@MaterialExpressionMaterialFunctionCall_45.Blended Material
   Layer Snow=@MaterialExpressionMaterialFunctionCall_11.Blended Material
  ) {layers [
    {
     blendType=LB_WeightBlend
     constHeightInput=0
     layerName=Grass_Only
     previewWeight=1
     constLayerInput=<0 0 0>
     heightInput {expression=null, inputName=None, mask=0, maskA=0, maskB=0, maskG=0, maskR=0, outputIndex=1}
     layerInput {
      expression=&self:MaterialExpressionMaterialFunctionCall_22
      inputName=None
      mask=0
      maskA=0
      maskB=0
      maskG=0
      maskR=0
      outputIndex=2
     }
    }
    {
     blendType=LB_WeightBlend
     constHeightInput=0
     layerName=Grass_Topology
     previewWeight=1
     constLayerInput=<0 0 0>
     heightInput {expression=null, inputName=None, mask=0, maskA=0, maskB=0, maskG=0, maskR=0, outputIndex=0}
     layerInput {
      expression=&self:MaterialExpressionMaterialFunctionCall_31
      inputName=None
      mask=0
      maskA=0
      maskB=0
      maskG=0
      maskR=0
      outputIndex=0
     }
    }
    {
     blendType=LB_WeightBlend
     constHeightInput=0
     layerName=Arid
     previewWeight=1
     constLayerInput=<0 0 0>
     heightInput {expression=null, inputName=None, mask=0, maskA=0, maskB=0, maskG=0, maskR=0, outputIndex=0}
     layerInput {
      expression=&self:MaterialExpressionStaticSwitchParameter_7
      inputName=None
      mask=0
      maskA=0
      maskB=0
      maskG=0
      maskR=0
      outputIndex=0
     }
    }
    {
     blendType=LB_WeightBlend
     constHeightInput=0
     layerName=Ice
     previewWeight=0
     constLayerInput=<0 0 0>
     heightInput {expression=null, inputName=None, mask=0, maskA=0, maskB=0, maskG=0, maskR=0, outputIndex=0}
     layerInput {
      expression=&self:MaterialExpressionMaterialFunctionCall_45
      inputName=None
      mask=0
      maskA=0
      maskB=0
      maskG=0
      maskR=0
      outputIndex=0
     }
    }
    {
     blendType=LB_WeightBlend
     constHeightInput=0
     layerName=Snow
     previewWeight=0
     constLayerInput=<0 0 0>
     heightInput {expression=null, inputName=None, mask=0, maskA=0, maskB=0, maskG=0, maskR=0, outputIndex=0}
     layerInput {
      expression=&self:MaterialExpressionMaterialFunctionCall_11
      inputName=None
      mask=0
      maskA=0
      maskB=0
      maskG=0
      maskR=0
      outputIndex=0
     }
    }
   ]}
  n45 = MaterialFunctionCall!(WorldPosition) {materialFunction=/FigmentEnvironment/LM_8_1_1/Athena/Environments/Landscape/MaterialFunctions/Utilities/MF_Athena_Terrain_Coords}
  n46 = TextureSample(UVs=@n45.Result) {
   samplerSource=SSM_Wrap_WorldGroupSettings
   samplerType=SAMPLERTYPE_LinearColor
   texture=/FigmentEnvironment/LM_8_1_1/Athena/Environments/Landscape/Textures/T_Athena_Terrain_RockColor
  }
  n47 = TextureSample(UVs=@n45.Result) {
   samplerSource=SSM_Wrap_WorldGroupSettings
   samplerType=SAMPLERTYPE_LinearColor
   texture=/FigmentEnvironment/LM_8_1_1/Athena/Environments/Landscape/Textures/T_Athena_Terrain_GroundColor
  }
  n48 = LinearInterpolate(A=@n46.RGB, B=@n47.RGB, Alpha=@n39.Alpha) {desc=Uses 2 color maps for better accuracy on height blend. This saves having to use layers for Biomes}
  n49 = CollectionParameter() {collection=&MPC_Landscape, parameterName=ColormapDivisor}
  n50 = LandscapeLayerSwitch(LayerNotUsed=MaterialFunctionCall(In=@n44, Color=Multiply(A=@n48, B=@n49)) {materialFunction=&MF_Athena_Material_BaseMultiply}.Result) {parameterName=Arid_Grass_Light}
  n51 = LandscapeLayerSwitch(LayerNotUsed=@n44) {desc=Low/Medium, parameterName=Arid_Grass_Light}
  n52 = MaterialFunctionCall!(In=@n50, Material_Low=@n51, Golf_Fairway_Mask=LandscapeLayerSample() {parameterName=Golf_Fairway, previewWeight=1}) {materialFunction=/FigmentEnvironment/LM_8_1_1/Athena/Environments/Landscape/MaterialFunctions/MF_LawnGrassColoration}
  n53 = Multiply(
   A=GetMaterialAttributes(None=LandscapeLayerSwitch(LayerUsed=@n52.Material_Low, LayerNotUsed=@n51) {desc=Low/Medium, parameterName=Golf_Fairway}) {attributeGetTypes [#1]}.Base Color
   B=Multiply(
    A=@n35.RGB
    B=CollectionParameter() {collection=&MPC_Landscape, parameterName=ColormapDivisor}
   )
  )
  n54 = MaterialFunctionCall() {materialFunction=&MF_Athena_Snow_01}
  n55 = Multiply(A=@n54.BaseColor, B=Constant3Vector() {constant=<rgba 0.6927 0.8414 1 0>}) {constA=0.6, constB=0.6}
  n56 = TextureSample(UVs=@n34) {
   samplerSource=SSM_Wrap_WorldGroupSettings
   samplerType=SAMPLERTYPE_LinearColor
   texture=/FigmentEnvironment/LM_8_1_1/Athena/Environments/Landscape/Textures/T_Athena_Terrain_Topo_Mask
  }
  n57 = MaterialFunctionCall!(
   Use_ExtraSnowMask
   UV_WorldMask=LandscapeLayerCoords() {mappingScale=1525}
   SnowStartBiasOverride
   UseHeightLerp
   LandscapeBlend_Height=@n54.Snow_Mask_Height
   Landscape_RockMask=OneMinus(None=@n16.Alpha)
   Snow_Mask
   Simple_UseWorldZNormal
   Simple_WorldNormal
   WorldPosition
  ) {materialFunction=/FigmentEnvironment/LM_8_1_1/Athena/Environments/Landscape/MaterialFunctions/Snow/MF_Athena_Holiday_SnowFall}
  n58 = StaticSwitchParameter(
   True=LinearInterpolate(A=@n53, B=LinearInterpolate(A=@n54.BaseColor, B=@n55, Alpha=@n56.B), Alpha=@n57.SnowMask_RockSubtract) {desc=Low snow blend}
   False=@n53
  ) {parameterName=Enable_HolidaySnow}
  n59 = MaterialFunctionCall!(UVs=LandscapeLayerCoords() {mappingScale=4}) {materialFunction=/FigmentEnvironment/LM_8_1_1/Athena/Environments/Landscape/MaterialFunctions/Standard/MF_Athena_Forest_01}
  n60 = LinearInterpolate(
   A=@n58
   B=@n59.BaseColor
   Alpha=MaterialFunctionCall(In=LandscapeLayerSample() {parameterName=Forest, previewWeight=1}, Contrast=Constant() {r=1}) {materialFunction=&CheapContrast}.Result
  )
  n61 = MaterialFunctionCall() {materialFunction=&MF_Athena_RoadGravel}
  n62 = Multiply(A=LinearInterpolate(A=Constant() {r=-0.25}, B=Constant() {r=1}, Alpha=@n61.Gravel_NoColor), B=@n4)
  n63 = MaterialFunctionCall(In=LandscapeLayerSample() {parameterName=Mud}, Contrast=Constant() {r=1}) {materialFunction=&CheapContrast}
  n64 = LandscapeLayerSample() {parameterName=Road, previewWeight=1}
  n65 = OneMinus(None=@MaterialExpressionMaterialFunctionCall_22.Height)
  n66 = ComponentMask(None=CollectionParameter() {collection=&MPC_Landscape, parameterName=Arid_Gravel_Tint}) {b=true, g=true, r=true}
  n67 = LandscapeLayerSwitch(
   LayerUsed=LinearInterpolate(A=@n61.Gravel, B=Multiply(A=@n61.Gravel_NoColor, B=@n66), Alpha=LandscapeLayerSample() {parameterName=Arid})
   LayerNotUsed=@n61.Gravel
  ) {parameterName=Arid}
  n68 = CollectionParameter() {collection=&MPC_Landscape, parameterName=Winter_Gravel_Tint}
  n69 = LandscapeLayerSample() {parameterName=Snow}
  n70 = LandscapeLayerSample() {parameterName=Ice}
  n71 = LandscapeLayerSample() {parameterName=SnowClumps}
  n72 = Clamp(None=Add(A=Add(A=@n69, B=@n70), B=@n71))
  n73 = LinearInterpolate(A=@n67, B=Multiply(A=@n61.Gravel_NoColor, B=ComponentMask(None=@n68) {b=true, g=true, r=true}), Alpha=@n72)
  n74 = LandscapeLayerSwitch(LayerUsed=@n73, LayerNotUsed=LandscapeLayerSwitch(LayerUsed=@n73, LayerNotUsed=@n67) {parameterName=Snow}) {parameterName=Ice}
  n75 = LinearInterpolate(
   A=LinearInterpolate(
    A=LinearInterpolate(A=@n60, B=@n62, Alpha=@n63.Result)
    B=@n61.Road
    Alpha=MaterialFunctionCall(Transition Phase=@n64, Height Texture=@n65) {materialFunction=&HeightLerp}.Lerp Alpha No Contrast
   )
   B=@n74
   Alpha=@n8
  )
  n76 = Saturate(None=CollectionParameter() {collection=&MPC_Landscape, parameterName=DisplayLowDetailShader})
  n77 = Saturate(None=CollectionParameter() {collection=&MPC_Landscape, parameterName=DisplayLowDetailShader})
  n78 = LandscapeLayerSwitch(LayerUsed=@n52.Material_Standard, LayerNotUsed=@n50) {parameterName=Golf_Fairway}
  n79 = GetMaterialAttributes(None=@n78) {attributeGetTypes [
    #3
    #2
    #1
    #5
   ]}
  n80 = MaterialFunctionCall(Transition Phase=LandscapeLayerSample() {parameterName=Forest, previewWeight=1}, Height Texture=@n59.Height) {materialFunction=&HeightLerp}
  n81 = LinearInterpolate(
   A=StaticSwitchParameter(True=LinearInterpolate(A=@n79.Base Color, B=@n53, Alpha=@n77), False=@n79.Base Color) {group=Utilities, parameterName=EnableQuickColorRender}
   B=@n59.BaseColor
   Alpha=@n80.Lerp Alpha No Contrast
  )
  n82 = OneMinus(None=@MaterialExpressionMaterialFunctionCall_22.Height)
  n83 = LandscapeLayerSample() {parameterName=Arid}
  n84 = MaterialFunctionCall(
   Transition Phase=LandscapeLayerSample() {parameterName=Mud}
   Height Texture=LandscapeLayerSwitch(LayerUsed=LinearInterpolate(A=@n82, B=@n61.Gravel_Height, Alpha=@n83) {desc=Change heightmap used for blend in Arid}, LayerNotUsed=@n82) {parameterName=Arid}
   Contrast=Constant() {r=0.2}
  ) {materialFunction=&HeightLerp}
  n85 = MaterialFunctionCall(
   Transition Phase=Power(Base=LandscapeLayerSample() {parameterName=Road, previewWeight=1})
   Height Texture=@n61.Road_Height
   Contrast=Constant() {r=5}
  ) {materialFunction=&HeightLerp}
  n86 = Multiply(A=LandscapeLayerSample() {parameterName=Road, previewWeight=1}, B=Constant() {r=1.2})
  n87 = Constant() {r=0.25}
  n88 = MaterialFunctionCall(
   Transition Phase=Clamp(None=Subtract(A=@n86, B=@n71))
   Height Texture=LinearInterpolate(A=Multiply(A=@n65, B=@n87), B=@n14.Height, Alpha=@n72)
   Contrast=Constant() {r=2}
  ) {materialFunction=&HeightLerp}
  n89 = MaterialFunctionCall(
   Transition Phase=LandscapeLayerSample() {parameterName=Gravel, previewWeight=1}
   Height Texture=LinearInterpolate(
    A=@n61.Gravel_Height
    B=Constant()
    Alpha=MaterialFunctionCall!(Distance) {materialFunction=/FigmentEnvironment/LM_8_1_1/Athena/Environments/Landscape/MaterialFunctions/MF_TerrainDistanceFade}.Result
   ) {constB=0.5}
  ) {materialFunction=&HeightLerp}
  n90 = LinearInterpolate(
   A=LinearInterpolate(A=LinearInterpolate(A=@n81, B=@n62, Alpha=@n84.Alpha), B=LinearInterpolate(A=@n62, B=@n61.Road, Alpha=@n85.Alpha), Alpha=@n88.Alpha)
   B=@n74
   Alpha=@n89.Lerp Alpha No Contrast
  )
  n91 = LinearInterpolate(
   A=StaticSwitchParameter(True=LinearInterpolate(A=@n90, B=@n75, Alpha=@n76), False=@n90) {group=Utilities, parameterName=EnableQuickColorRender}
   B=@n90
   Alpha=LandscapeLayerSample() {parameterName=NoTopo}
  )
  MaterialExpressionConstant_4 = Constant() {r=1}
  MaterialExpressionLinearInterpolate_41 = LinearInterpolate(
   A=ComponentMask(None=CollectionParameter() {collection=&MPC_Landscape, parameterName=Arid_Rock_01_Tint}) {b=true, g=true, r=true}
   B=Constant3Vector() {constant=<rgba 1 0.7455 0.5781 0>}
   Alpha=CollectionParameter() {collection=&MPC_Landscape, parameterName=Display_Combined}
  )
  n92 = LandscapeLayerBlend(
   Layer Grass_Only=@MaterialExpressionConstant_4
   Layer Grass_Topology=@MaterialExpressionConstant_4
   Layer Arid=@MaterialExpressionLinearInterpolate_41
   Layer Ice=@MaterialExpressionConstant_4
   Layer Snow=@MaterialExpressionConstant_4
  ) {layers [
    {
     blendType=LB_WeightBlend
     constHeightInput=0
     layerName=Grass_Only
     previewWeight=1
     constLayerInput=<0 0 0>
     heightInput {expression=null, inputName=None, mask=0, maskA=0, maskB=0, maskG=0, maskR=0, outputIndex=0}
     layerInput {
      expression=&self:MaterialExpressionConstant_4
      inputName=None
      mask=0
      maskA=0
      maskB=0
      maskG=0
      maskR=0
      outputIndex=0
     }
    }
    {
     blendType=LB_WeightBlend
     constHeightInput=0
     layerName=Grass_Topology
     previewWeight=1
     constLayerInput=<0 0 0>
     heightInput {expression=null, inputName=None, mask=0, maskA=0, maskB=0, maskG=0, maskR=0, outputIndex=0}
     layerInput {
      expression=&self:MaterialExpressionConstant_4
      inputName=None
      mask=0
      maskA=0
      maskB=0
      maskG=0
      maskR=0
      outputIndex=0
     }
    }
    {
     blendType=LB_WeightBlend
     constHeightInput=0
     layerName=Arid
     previewWeight=1
     constLayerInput=<0 0 0>
     heightInput {expression=null, inputName=None, mask=0, maskA=0, maskB=0, maskG=0, maskR=0, outputIndex=0}
     layerInput {
      expression=&self:MaterialExpressionLinearInterpolate_41
      inputName=None
      mask=0
      maskA=0
      maskB=0
      maskG=0
      maskR=0
      outputIndex=0
     }
    }
    {
     blendType=LB_WeightBlend
     constHeightInput=0
     layerName=Ice
     previewWeight=0
     constLayerInput=<0 0 0>
     heightInput {expression=null, inputName=None, mask=0, maskA=0, maskB=0, maskG=0, maskR=0, outputIndex=0}
     layerInput {
      expression=&self:MaterialExpressionConstant_4
      inputName=None
      mask=0
      maskA=0
      maskB=0
      maskG=0
      maskR=0
      outputIndex=0
     }
    }
    {
     blendType=LB_WeightBlend
     constHeightInput=0
     layerName=Snow
     previewWeight=0
     constLayerInput=<0 0 0>
     heightInput {expression=null, inputName=None, mask=0, maskA=0, maskB=0, maskG=0, maskR=0, outputIndex=0}
     layerInput {
      expression=&self:MaterialExpressionConstant_4
      inputName=None
      mask=0
      maskA=0
      maskB=0
      maskG=0
      maskR=0
      outputIndex=0
     }
    }
   ]}
  n93 = CollectionParameter() {collection=&MPC_Landscape, parameterName=MountainRockTint}
  n94 = LinearInterpolate(
   A=LinearInterpolate(A=@n92, B=ComponentMask(None=@n93) {b=true, g=true, r=true}, Alpha=LandscapeLayerSample() {parameterName=Mountain_Grass})
   B=ComponentMask(None=CollectionParameter() {collection=&MPC_Landscape, parameterName=FarmRockTint}) {b=true, g=true, r=true}
   Alpha=LandscapeLayerSample() {parameterName=Farm_Grass}
  )
  MaterialExpressionConstant_3 = Constant() {r=1}
  MaterialExpressionMultiply_2 = Multiply(
   A=MaterialFunctionCall() {materialFunction=/FigmentEnvironment/LM_8_1_1/Athena/Environments/Landscape/MaterialFunctions/MF_Athena_ZSandColor}.Result
   B=ComponentMask(None=CollectionParameter() {collection=&MPC_Landscape, parameterName=Arid_Sand_Tint}) {b=true, g=true, r=true}
  )
  n95 = LandscapeLayerBlend(
   Layer Grass_Only=@MaterialExpressionConstant_3
   Layer Grass_Topology=@MaterialExpressionConstant_3
   Layer Arid=@MaterialExpressionMultiply_2
   Layer Ice=@MaterialExpressionConstant_3
   Layer Snow=@MaterialExpressionConstant_3
  ) {layers [
    {
     blendType=LB_WeightBlend
     constHeightInput=0
     layerName=Grass_Only
     previewWeight=1
     constLayerInput=<0 0 0>
     heightInput {expression=null, inputName=None, mask=0, maskA=0, maskB=0, maskG=0, maskR=0, outputIndex=0}
     layerInput {
      expression=&self:MaterialExpressionConstant_3
      inputName=None
      mask=0
      maskA=0
      maskB=0
      maskG=0
      maskR=0
      outputIndex=0
     }
    }
    {
     blendType=LB_WeightBlend
     constHeightInput=0
     layerName=Grass_Topology
     previewWeight=1
     constLayerInput=<0 0 0>
     heightInput {expression=null, inputName=None, mask=0, maskA=0, maskB=0, maskG=0, maskR=0, outputIndex=0}
     layerInput {
      expression=&self:MaterialExpressionConstant_3
      inputName=None
      mask=0
      maskA=0
      maskB=0
      maskG=0
      maskR=0
      outputIndex=0
     }
    }
    {
     blendType=LB_WeightBlend
     constHeightInput=0
     layerName=Arid
     previewWeight=1
     constLayerInput=<0 0 0>
     heightInput {expression=null, inputName=None, mask=0, maskA=0, maskB=0, maskG=0, maskR=0, outputIndex=0}
     layerInput {
      expression=&self:MaterialExpressionMultiply_2
      inputName=None
      mask=0
      maskA=0
      maskB=0
      maskG=0
      maskR=0
      outputIndex=0
     }
    }
    {
     blendType=LB_WeightBlend
     constHeightInput=0
     layerName=Ice
     previewWeight=0
     constLayerInput=<0 0 0>
     heightInput {expression=null, inputName=None, mask=0, maskA=0, maskB=0, maskG=0, maskR=0, outputIndex=0}
     layerInput {
      expression=&self:MaterialExpressionConstant_3
      inputName=None
      mask=0
      maskA=0
      maskB=0
      maskG=0
      maskR=0
      outputIndex=0
     }
    }
    {
     blendType=LB_WeightBlend
     constHeightInput=0
     layerName=Snow
     previewWeight=0
     constLayerInput=<0 0 0>
     heightInput {expression=null, inputName=None, mask=0, maskA=0, maskB=0, maskG=0, maskR=0, outputIndex=0}
     layerInput {
      expression=&self:MaterialExpressionConstant_3
      inputName=None
      mask=0
      maskA=0
      maskB=0
      maskG=0
      maskR=0
      outputIndex=0
     }
    }
   ]}
  n96 = CollectionParameter() {collection=&MPC_Landscape, parameterName=MountainGrassTint}
  n97 = Subtract(A=LandscapeLayerSample() {parameterName=Mountain_Grass}, B=LandscapeLayerSample() {parameterName=Snow})
  n98 = LinearInterpolate(
   A=LinearInterpolate(A=@n95, B=ComponentMask(None=@n96) {b=true, g=true, r=true}, Alpha=Clamp(None=@n97))
   B=ComponentMask(None=CollectionParameter() {collection=&MPC_Landscape, parameterName=FarmGrassTint}) {b=true, g=true, r=true}
   Alpha=LandscapeLayerSample() {parameterName=Farm_Grass}
  )
  n99 = LinearInterpolate(
   A=LinearInterpolate(
    A=@n94
    B=@n98
    Alpha=CollectionParameter() {collection=&MPC_Landscape, parameterName=Display_GroundColors}
   )
   B=LinearInterpolate(A=@n94, B=@n98, Alpha=@n16.Alpha_ColorMap)
   Alpha=CollectionParameter() {collection=&MPC_Landscape, parameterName=Display_Combined}
  )
  n100 = TextureSample(UVs=MaterialFunctionCall() {materialFunction=&MF_Athena_Terrain_Coords}.Result) {
   samplerSource=SSM_Wrap_WorldGroupSettings
   samplerType=SAMPLERTYPE_LinearColor
   texture=&T_Athena_Terrain_Topo_Mask
  }
  n101 = Multiply(A=Multiply(A=@n99) {constB=2}, B=@n100.G)
  n102 = LandscapeLayerSample() {parameterName=NoTopo}
  n103 = Constant() {r=1}
  n104 = ScalarParameter() {defaultValue=2, parameterName=RockEdgeDesat}
  n105 = CollectionParameter() {collection=&MPC_Landscape, parameterName=Display_Combined}
  n106 = OneMinus(None=MaterialFunctionCall(Height=MaterialFunctionCall() {materialFunction=&MF_Athena_Grass_01}.Height) {materialFunction=&MF_Athena_SlopeBlend}.Alpha_NoContrast)
  n107 = VertexNormalWS()
  n108 = MaterialFunctionCall(
   In=Abs(None=ComponentMask(None=@n107) {b=true})
   Contrast=MaterialFunctionCall() {materialFunction=&MF_Athena_Rock_01}.Proj_Contrast
  ) {materialFunction=&CheapContrast}
  n109 = LinearInterpolate(A=@n106, B=Min(A=@n106, B=ComponentMask(None=@n108.Result) {r=true}), Alpha=@n105)
  n110 = CollectionParameter() {collection=&MPC_Landscape, parameterName=Display_GroundColors}
  n111 = LinearInterpolate(
   B=LinearInterpolate(
    A=LinearInterpolate(A=@n103, B=@n104, Alpha=@n105)
    B=ScalarParameter() {parameterName=RockEdgeDesat_Arid}
    Alpha=LandscapeLayerSample() {parameterName=Arid}
   )
   Alpha=Clamp(None=LinearInterpolate(B=@n109, Alpha=@n110))
  )
  n112 = LinearInterpolate(
   A=Desaturation(None=LinearInterpolate(A=@n101, B=@n99, Alpha=@n102), Fraction=Clamp(None=@n111)) {desc=To Render Colormap}
   B=Constant() {r=1.375}
   Alpha=MaterialFunctionCall(In=LandscapeLayerSample() {parameterName=Golf_Green}, Contrast=ScalarParameter() {parameterName=Golf_Mobile_LayerContrast}) {materialFunction=&CheapContrast}.Result
  )
  n113 = Subtract(A=@n100.B, B=LandscapeLayerSample() {parameterName=NoTopo})
  n114 = LinearInterpolate(A=@n112, B=Multiply(A=@n112, B=Constant3Vector() {constant=<rgba 0.8536 0.9838 1.1 1>}), Alpha=Clamp(None=Multiply(A=@n113, B=@n16.Alpha)))
  n115 = CollectionParameter() {collection=&MPC_Landscape, parameterName=ColormapDivisor}
  n116 = CollectionParameter() {collection=&MPC_Landscape, parameterName=GrassEdgeZ}
  n117 = ComponentMask(None=VertexNormalWS()) {b=true}
  n118 = CollectionParameter() {collection=&MPC_Landscape, parameterName=GrassEdgeDesatWidth}
  n119 = MaterialFunctionCall!(UVW=Desaturation(None=Divide(A=@n114, B=@n115), Fraction=SphereMask(A=@n116, B=@n117, Radius=@n118) {hardnessPercent=0})) {materialFunction=/Engine/ArtTools/RenderToTexture/MaterialFunctions/MarkRedOutside0-1Range}
  n120 = Saturate(None=CollectionParameter() {collection=&MPC_Landscape, parameterName=RenderColorMap})
  MaterialExpressionConstant_9 = Constant()
  MaterialExpressionConstant3Vector_1 = Constant3Vector() {constant=<rgba 0 0 1 0>}
  MaterialExpressionConstant3Vector_2 = Constant3Vector() {desc=Farm, constant=<rgba 1 0 0 0>}
  MaterialExpressionConstant3Vector_3 = Constant3Vector() {desc=Mountain, constant=<rgba 0 1 0 0>}
  n121 = LandscapeLayerBlend!(
   Layer Grass_Only=@MaterialExpressionConstant_9
   Layer Grass_Topology=@MaterialExpressionConstant_9
   Layer Arid=@MaterialExpressionConstant3Vector_1
   Layer Farm_Grass=@MaterialExpressionConstant3Vector_2
   Layer Mountain_Grass=@MaterialExpressionConstant_9
   Layer Ice=@MaterialExpressionConstant3Vector_3
   Layer Snow=@MaterialExpressionConstant3Vector_3
  ) {layers [
    {
     blendType=LB_WeightBlend
     constHeightInput=0
     layerName=Grass_Only
     previewWeight=1
     constLayerInput=<0 0 0>
     heightInput {expression=null, inputName=None, mask=0, maskA=0, maskB=0, maskG=0, maskR=0, outputIndex=0}
     layerInput {
      expression=&self:MaterialExpressionConstant_9
      inputName=None
      mask=0
      maskA=0
      maskB=0
      maskG=0
      maskR=0
      outputIndex=0
     }
    }
    {
     blendType=LB_WeightBlend
     constHeightInput=0
     layerName=Grass_Topology
     previewWeight=1
     constLayerInput=<0 0 0>
     heightInput {expression=null, inputName=None, mask=0, maskA=0, maskB=0, maskG=0, maskR=0, outputIndex=0}
     layerInput {
      expression=&self:MaterialExpressionConstant_9
      inputName=None
      mask=0
      maskA=0
      maskB=0
      maskG=0
      maskR=0
      outputIndex=0
     }
    }
    {
     blendType=LB_WeightBlend
     constHeightInput=0
     layerName=Arid
     previewWeight=1
     constLayerInput=<0 0 0>
     heightInput {expression=null, inputName=None, mask=0, maskA=0, maskB=0, maskG=0, maskR=0, outputIndex=0}
     layerInput {
      expression=&self:MaterialExpressionConstant3Vector_1
      inputName=None
      mask=0
      maskA=0
      maskB=0
      maskG=0
      maskR=0
      outputIndex=0
     }
    }
    {
     blendType=LB_WeightBlend
     constHeightInput=0
     layerName=Farm_Grass
     previewWeight=1
     constLayerInput=<0 0 0>
     heightInput {expression=null, inputName=None, mask=0, maskA=0, maskB=0, maskG=0, maskR=0, outputIndex=0}
     layerInput {
      expression=&self:MaterialExpressionConstant3Vector_2
      inputName=None
      mask=0
      maskA=0
      maskB=0
      maskG=0
      maskR=0
      outputIndex=0
     }
    }
    {
     blendType=LB_WeightBlend
     constHeightInput=0
     layerName=Mountain_Grass
     previewWeight=1
     constLayerInput=<0 0 0>
     heightInput {expression=null, inputName=None, mask=0, maskA=0, maskB=0, maskG=0, maskR=0, outputIndex=0}
     layerInput {
      expression=&self:MaterialExpressionConstant_9
      inputName=None
      mask=0
      maskA=0
      maskB=0
      maskG=0
      maskR=0
      outputIndex=0
     }
    }
    {
     blendType=LB_WeightBlend
     constHeightInput=0
     layerName=Ice
     previewWeight=0
     constLayerInput=<0 0 0>
     heightInput {expression=null, inputName=None, mask=0, maskA=0, maskB=0, maskG=0, maskR=0, outputIndex=0}
     layerInput {
      expression=&self:MaterialExpressionConstant3Vector_3
      inputName=None
      mask=0
      maskA=0
      maskB=0
      maskG=0
      maskR=0
      outputIndex=0
     }
    }
    {
     blendType=LB_WeightBlend
     constHeightInput=0
     layerName=Snow
     previewWeight=0
     constLayerInput=<0 0 0>
     heightInput {expression=null, inputName=None, mask=0, maskA=0, maskB=0, maskG=0, maskR=0, outputIndex=0}
     layerInput {
      expression=&self:MaterialExpressionConstant3Vector_3
      inputName=None
      mask=0
      maskA=0
      maskB=0
      maskG=0
      maskR=0
      outputIndex=0
     }
    }
   ]}
  n122 = CollectionParameter() {collection=&MPC_Landscape, parameterName=DisplayBiomeRGB}
  n123 = StaticSwitchParameter(True=LinearInterpolate(A=LinearInterpolate(A=@n91, B=@n119.Result, Alpha=@n120), B=@n121, Alpha=Saturate(None=@n122)), False=@n91) {group=Utilities, parameterName=EnableQuickColorRender}
  n124 = LinearInterpolate(A=@n58, B=@n59.BaseColor, Alpha=@n80.Lerp Alpha No Contrast)
  n125 = LinearInterpolate(
   A=LinearInterpolate(A=LinearInterpolate(A=@n124, B=@n62, Alpha=@n84.Alpha), B=@n61.Road, Alpha=@n88.Alpha)
   B=@n74
   Alpha=@n89.Lerp Alpha No Contrast
  )
  n126 = QualitySwitch(Default=@n123, Low=@n125, High=@n123, Medium=@n125)
  n127 = FeatureLevelSwitch(Default=@n126, ES2=@n75, ES3_1=@n75, SM4_REMOVED=@n126, SM5=@n126, SM6=@n126)
  n128 = Multiply(A=ComponentMask(None=@n127) {r=true}, B=Constant() {r=1})
  n129 = FeatureLevelSwitch(
   Default=LinearInterpolate(A=LinearInterpolate(A=@n27, B=@n28, Alpha=@n29), B=Clamp(None=@n31.Specular), Alpha=LandscapeLayerSample() {parameterName=Snow})
   ES2=@n128
   ES3_1=@n128
  )
  n130 = Constant3Vector() {constant=<rgba 0 0 1 0>}
  n131 = Add(
   A=Add(A=LandscapeLayerSample() {parameterName=Road, previewWeight=1}, B=LandscapeLayerSample() {parameterName=Mud, previewWeight=1})
   B=LandscapeLayerSample() {parameterName=Gravel, previewWeight=1}
  )
  n132 = Clamp(None=Add(A=@n131, B=LandscapeLayerSample() {parameterName=NoTopo, previewWeight=1}))
  n133 = WorldPosition() {worldPositionShaderOffset=WPT_ExcludeAllShaderOffsets}
  n134 = TextureSample(UVs=Divide(A=ComponentMask(None=@n133.XYZ) {g=true, r=true}, B=Constant() {r=25000})) {
   samplerSource=SSM_Wrap_WorldGroupSettings
   samplerType=SAMPLERTYPE_Normal
   texture=/Game/ContentCreationTools/TextureCreation/Textures/T_BPCreated_MacroNormal_01
  }
  n135 = Constant() {r=0.625}
  n136 = Constant() {r=0.25}
  n137 = Constant() {r=0.9}
  n138 = FeatureLevelSwitch(Default=@n136, ES2=@n137, ES3_1=@n137, SM4_REMOVED=@n136)
  n139 = Multiply(
   A=LinearInterpolate(
    A=Constant3Vector() {constant=<rgba 0 0 1 0>}
    B=@n134.RGB
    Alpha=MaterialFunctionCall() {materialFunction=&MF_TerrainDistanceFade}.Result
   )
   B=MaterialFunctionCall(X=@n135, Y=@n135, Z=QualitySwitch(Default=@n138, Low=@n137, Medium=@n137)) {materialFunction=&MakeFloat3}.Result
  )
  n140 = Add(A=Multiply(A=LinearInterpolate(A=@n79.Normal, B=@n130, Alpha=@n132), B=Constant3Vector() {constant=<rgba 1 1 0.5 0>}), B=@n139)
  n141 = Add(A=LandscapeLayerSample() {parameterName=Gravel}, B=LandscapeLayerSample() {parameterName=Forest})
  n142 = LandscapeLayerSample() {parameterName=Mud}
  n143 = Subtract(
   A=Add(A=LandscapeLayerSample() {parameterName=Ice}, B=@n57.SnowMask_RockSubtract)
   B=Add(A=Add(A=@n141, B=@n142), B=LandscapeLayerSample() {parameterName=Road})
  ) {desc=Mask out paths and dirt from emmisive}
  n144 = LinearInterpolate(A=Add(A=@n30.Normal, B=@n139), B=@n30.Normal, Alpha=Clamp(None=Multiply(A=@n16.Alpha, B=@n143)))
  n145 = QualitySwitch(Default=@n140, Low=@n144, High=@n140, Medium=@n144)
  n146 = FeatureLevelSwitch(Default=@n145, ES2=@n139, ES3_1=@n139, SM4_REMOVED=@n145, SM5=@n145, SM6=@n145)
  n147 = Constant() {r=1}
  n148 = Clamp(None=Add(A=@n131, B=LandscapeLayerSample() {parameterName=Forest, previewWeight=1}))
  n149 = LinearInterpolate(A=@n79.Roughness, Alpha=@n148)
  n150 = QualitySwitch(Default=@n149, Low=@n147, High=@n149, Medium=@n147)
  n151 = FeatureLevelSwitch(Default=@n150, ES2=@n147, ES3_1=@n147, SM4_REMOVED=@n150, SM5=@n150, SM6=@n150)
  n152 = SetMaterialAttributes!(MaterialAttributes, Base Color=@n127, Specular=@n129, Roughness=@n151, Opacity Mask=@n26, Normal=@n146, Emissive Color=@n24) {attributeSetTypes [
    #1
    #2
    #3
    #4
    #5
    #6
   ]}
  MaterialFunctionCall(NoRVT=@n152, RVT=@n25) {materialFunction=&FN_RVT_Switch_Material}
  n153 = GetMaterialAttributes(None=@n54.Material) {attributeGetTypes [#1]}
  n154 = ComponentMask(None=CollectionParameter() {collection=&MPC_Landscape, parameterName=Snow_CavTint}) {b=true, g=true, r=true}
  n155 = TextureSample(UVs=MaterialFunctionCall() {materialFunction=&MF_Athena_Terrain_Coords}.Result) {
   samplerSource=SSM_Wrap_WorldGroupSettings
   samplerType=SAMPLERTYPE_LinearColor
   texture=&T_Athena_Terrain_Topo_Mask
  }
  n156 = MaterialFunctionCall(Distance=Constant() {r=25000}) {materialFunction=&MF_TerrainDistanceFade}
  n157 = SetMaterialAttributes!(
   MaterialAttributes=@n153.MaterialAttributes
   Base Color=LinearInterpolate(A=@n153.Base Color, B=Multiply(A=@n153.Base Color, B=@n154), Alpha=Multiply(A=@n155.B, B=@n156.Result))
  ) {attributeSetTypes [#1]}
  NamedRerouteDeclaration(None=StaticSwitchParameter(
    True=MaterialFunctionCall(Base Material=@n78, Top Material=@n157, Alpha=@n57.SnowMask_RockSubtract) {materialFunction=&MatLayerBlend_Standard}.Blended Material
    False=@n78
   ) {parameterName=Enable_HolidaySnow}) {name=grass, nodeColor=<rgba 0.3828 1 0 1>}
  NamedRerouteDeclaration(None=LinearInterpolate(A=@n79.Specular, B=ScalarParameter() {parameterName=NonGrassSpecular}, Alpha=@n148)) {name=Spec, nodeColor=<rgba 1 0.9609 0 1>}
  NamedRerouteUsage()
  Reroute(None=@MaterialExpressionMaterialFunctionCall_22.Result)
  n158 = LandscapeLayerSample() {parameterName=NoFoliage, previewWeight=1}
  Reroute(None=Subtract(A=LandscapeLayerSample() {parameterName=CraterOvergrowth}, B=@n158)) {desc=Farm}
  n159 = LandscapeLayerSample() {parameterName=Grass_Only, previewWeight=1}
  n160 = LandscapeLayerSample() {parameterName=Gravel}
  n161 = Constant() {r=-65}
  n162 = Constant() {r=50}
  n163 = Add(A=@n16.Alpha, B=Constant())
  n164 = OneMinus(None=Clamp(None=LinearInterpolate(A=@n161, B=@n162, Alpha=@n163)))
  n165 = Round(None=@n164)
  n166 = LandscapeLayerSample() {parameterName=Mud}
  n167 = LandscapeLayerSample() {parameterName=Road, previewWeight=1}
  n168 = Add(A=Add(A=Add(A=@n160, B=@n165), B=@n166), B=@n167)
  n169 = LandscapeLayerSample() {parameterName=NoFoliage}
  n170 = Add(A=Ceil(None=LandscapeLayerSample() {parameterName=Golf_Fairway}), B=Ceil(None=LandscapeLayerSample() {parameterName=Golf_Green}))
  n171 = Add(A=Add(A=Add(A=@n168, B=@n169), B=LandscapeLayerSample() {parameterName=Forest}), B=Round(None=@n170))
  n172 = Subtract(A=Add(A=LandscapeLayerSample() {parameterName=Grass_Topology, previewWeight=1}, B=Round(None=@n159)), B=@n171)
  Reroute(None=@n172) {desc=grass}
  n173 = LandscapeLayerSample() {parameterName=Flowers, previewWeight=1}
  n174 = Add(A=LandscapeLayerSample() {parameterName=Snow}, B=LandscapeLayerSample() {parameterName=Ice})
  Reroute(None=Subtract(A=Subtract(A=@n173, B=@n158), B=@n174)) {desc=flowers}
  n175 = LandscapeLayerSample() {parameterName=Farm_Grass}
  n176 = LandscapeLayerSample() {parameterName=Mountain_Grass}
  n177 = Add(A=Add(A=@n172, B=Subtract(A=@n175, B=@n171)), B=Subtract(A=Add(A=@n176, B=@n175), B=@n171))
  n178 = Round(None=LandscapeLayerSample() {parameterName=Arid})
  Reroute(None=Clamp(None=Add(A=@n177, B=@n178))) {desc=bushes}
  n179 = LandscapeLayerSample() {parameterName=Crater}
  Reroute(None=@n179) {desc=crater}
  Reroute(None=Subtract(A=Subtract(A=@n178, B=@n164) {desc=Subtract rock}, B=@n171)) {desc=arid}
  Reroute(None=Multiply(A=Subtract(A=@n174, B=@n171), B=@n173)) {desc=snow}
  n180 = Subtract(A=Subtract(A=LandscapeLayerSample() {parameterName=Gravel}, B=@n164), B=Multiply(B=@n179) {constA=10})
  n181 = LandscapeLayerSample() {parameterName=NoFoliage}
  n182 = LandscapeLayerSample() {parameterName=Ice}
  n183 = LandscapeLayerSample() {parameterName=Snow}
  Reroute(None=Subtract(A=Subtract(A=@n180, B=@n181), B=Add(A=@n182, B=@n183))) {desc=gravel}
  n184 = Add(A=Add(A=@n166, B=@n167), B=@n170)
  Reroute(None=Subtract(A=LandscapeLayerSample() {parameterName=Forest}, B=Add(A=@n184, B=@n169))) {desc=Forest}
  Reroute(None=@n24) {desc=emissive}
  Reroute(None=@n146) {desc=Normal}
  Reroute(None=@n151) {desc=roughness}
  Reroute(None=@n129) {desc=Specular}
  Reroute(None=@n127) {desc=Base Color}
  RuntimeVirtualTextureOutput!(BaseColor=@n127, Specular=@n129, Roughness=@n151, Normal=@n146, WorldHeight=WorldPosition().Z, Opacity, Mask=@n26, Displacement, Mask4)
  ScalarParameter()
  ScalarParameter() {defaultValue=306200, parameterName=TerrainSize}
  ScalarParameter() {defaultValue=-153000, parameterName=LandscapeMin, sortPriority=-153000}
  ScalarParameter() {defaultValue=306000, parameterName=LandscapeSize, sortPriority=-153000}
 ]
}
