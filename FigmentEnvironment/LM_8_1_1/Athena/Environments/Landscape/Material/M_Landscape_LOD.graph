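# Landscape LOD material for the Athena terrain in the FigmentEnvironment content, using default lit shading.
# A baked terrain LOD texture on landscape coordinates is tinted by parameter C, alongside scalar parameters R and S.
# A world-XY macro normal map fades from flat through a terrain distance function, then gets scaled by intensity.
# Quality and feature-level switches vary the Z scale and drop an added flat-normal term on lower tiers.
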Material /FigmentEnvironment/LM_8_1_1/Athena/Environments/Landscape/Material/M_Landscape_LOD
graph {
 properties {pixelDepthOffsetMode=PDOM_Legacy, usedShadingModels=Default Lit}
 nodes [
  n1 = Divide(
   A=ComponentMask(None=WorldPosition() {worldPositionShaderOffset=WPT_ExcludeAllShaderOffsets}.XYZ) {g=true, r=true}
   B=ScalarParameter() {defaultValue=25000, parameterName=MacroNormalScale}
  )
  n2 = ScalarParameter() {defaultValue=0.625, parameterName=MacroNormalIntensity}
  n3 = Constant() {r=0.25}
  n4 = Constant() {r=0.9}
  n5 = FeatureLevelSwitch(Default=@n3, ES2=@n4, ES3_1=@n4, SM4_REMOVED=@n3)
  n6 = Multiply(
   A=LinearInterpolate(
    A=Constant3Vector() {constant=<rgba 0 0 1 0>}
    B=TextureSample(UVs=@n1) {samplerType=SAMPLERTYPE_Normal, texture=/Game/ContentCreationTools/TextureCreation/Textures/T_BPCreated_MacroNormal_01}.RGB
    Alpha=MaterialFunctionCall!(Distance) {materialFunction=/FigmentEnvironment/LM_8_1_1/Athena/Environments/Landscape/MaterialFunctions/MF_TerrainDistanceFade}.Result
   )
   B=MaterialFunctionCall(X=@n2, Y=@n2, Z=QualitySwitch(Default=@n5, Low=@n4, Medium=@n4)) {materialFunction=&MakeFloat3}.Result
  )
  n7 = Add(A=Multiply(A=Constant3Vector() {constant=<rgba 0 0 1 0>}, B=Constant3Vector() {constant=<rgba 1 1 0.5 0>}), B=@n6)
  n8 = QualitySwitch(Default=@n7, Low=@n6, High=@n7, Medium=@n6)
  FeatureLevelSwitch(Default=@n8, ES2=@n6, ES3_1=@n6, SM4_REMOVED=@n8, SM5=@n8, SM6=@n8)
  Multiply(
   A=VectorParameter() {parameterName=C, defaultValue=<rgba 1 1 1 0>}.RGB
   B=TextureSample(UVs=LandscapeLayerCoords() {mappingScale=1525}) {texture=/FigmentEnvironment/LM_8_1_1/Athena/Environments/Landscape/Textures/T_Athena_Terrain_LOD}.RGB
  )
  ScalarParameter() {defaultValue=1, parameterName=R}
  ScalarParameter() {parameterName=S}
 ]
}
