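# Lava ground material for volcanic environment meshes, supporting Nanite and instancing, built from layered material-attribute functions.
# World-space noise drives height-lerped soil over lava ground, with switchable vertex-painted grass and normal-Z wall layers.
# An optional edge gradient uses UV1 or, with rock blend, object-relative height snapped to a 384-unit grid.
# Rock blend layers cliff textures by slope or vertex green, combining normals, before MF_CreativeMoveTool.
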Material /FigmentEnvironment/LM_8_1_1/Athena/Environments/Landscape/Material/M_LavaGround
graph {
 properties {
  bTangentSpaceNormal=false
  bUseMaterialAttributes=true
  bUsedWithInstancedStaticMeshes=true
  bUsedWithNanite=true
  ditheredLODTransition=true
  pixelDepthOffsetMode=PDOM_Legacy
  usedShadingModels=Default Lit
 }
 nodes [
  n1 = GetMaterialAttributes(None=MaterialFunctionCall() {materialFunction=/Game/Athena/Environments/Landscape/MaterialFunctions/Standard/MF_Athena_Grass_01}.Result) {attributeGetTypes [#1]}
  n2 = VertexColor()
  n3 = Multiply(
   A=@n1.Base Color
   B=LinearInterpolate(
    A=VectorParameter() {parameterName=ColorGrassTint, defaultValue=<rgba 0.4886 0.7656 0.3807 1>}.RGB
    B=VectorParameter() {parameterName=LavaGrasTint, defaultValue=<rgba 0.5521 0.2139 0.1141 1>}.RGB
    Alpha=@n2.B
   )
  )
  n4 = MaterialFunctionCall!(UV, WorldAlignedSize) {materialFunction=/Game/Environments/Sets/Vulcano/Materials/MF_LavaGround}
  n5 = ScalarParameter() {parameterName=SoilMin}
  n6 = ScalarParameter() {defaultValue=1.5828, parameterName=SoilMax}
  n7 = WorldPosition()
  n8 = TextureSample(UVs=Divide(A=ComponentMask(None=@n7.XYZ) {g=true, r=true}, B=ScalarParameter() {defaultValue=2655.5642, parameterName=LavaGround_BlendSize})) {texture=/Game/Global/Textures/Default/Noise/T_Tiling_Noise_01}
  n9 = MaterialFunctionCall(
   Transition Phase=Clamp(None=LinearInterpolate(A=@n5, B=@n6, Alpha=@n8.R))
   Height Texture=OneMinus(None=@n4.Height)
   Contrast=ScalarParameter() {defaultValue=0.0476, parameterName=MatlayerBlendcontrast}
  ) {materialFunction=&HeightLerp}
  n10 = VertexNormalWS()
  n11 = Clamp(None=LinearInterpolate(
    A=ScalarParameter() {defaultValue=-7.9409, parameterName=WallBlendMin}
    B=ScalarParameter() {defaultValue=9.2509, parameterName=WallBlendMax}
    Alpha=ComponentMask(None=@n10) {b=true}
   ))
  n12 = MaterialFunctionCall!(
   Base Material=@n4.Result
   Top Material=MaterialFunctionCall!(UVs) {materialFunction=/Game/Environments/Sets/Vulcano/Materials/MF_Lava_Soil}.Result
   Alpha=StaticSwitchParameter(True=Multiply(A=@n9.Alpha, B=@n11), False=@n9.Alpha) {parameterName=FadeSandonWalls}
  ) {materialFunction=/Engine/Functions/MaterialLayerFunctions/MatLayerBlend_Standard}
  n13 = ScalarParameter() {defaultValue=0.4595, parameterName=GrassBlendContrast}
  n14 = StaticSwitchParameter(
   True=MaterialFunctionCall(
    Base Material=SetMaterialAttributes!(MaterialAttributes=@n1.MaterialAttributes, Base Color=@n3) {attributeSetTypes [#1]}
    Top Material=@n12.Blended Material
    Alpha=MaterialFunctionCall(Transition Phase=@n2.A, Height Texture=@n4.Height, Contrast=@n13) {materialFunction=&HeightLerp}.Alpha
   ) {materialFunction=&MatLayerBlend_Standard}.Blended Material
   False=@n12.Blended Material
  ) {parameterName=UseGrassBlend}
  n15 = VertexNormalWS()
  n16 = Clamp(None=LinearInterpolate(
    A=ScalarParameter() {defaultValue=-7.9409, group=WallBlend, parameterName=WallBlendMin}
    B=ScalarParameter() {defaultValue=9.2509, group=WallBlend, parameterName=WallBlendMax}
    Alpha=ComponentMask(None=@n15) {b=true}
   ))
  n17 = StaticSwitchParameter(
   True=MaterialFunctionCall(
    Base Material=MaterialFunctionCall() {materialFunction=/Game/Environments/Sets/Vulcano/Materials/MF_LavaWalls}.Result
    Top Material=@n14
    Alpha=@n16
   ) {materialFunction=&MatLayerBlend_Standard}.Blended Material
   False=@n14
  ) {defaultValue=true, parameterName=UseWalls}
  n18 = TextureCoordinate() {uTiling=3, vTiling=3}
  MaterialFunctionCall(
   Base Material=@n17
   Top Material=MaterialFunctionCall!(
    Material=MaterialFunctionCall(UVs=@n18) {materialFunction=&MF_Lava_Soil}.Result
    Tint=VectorParameter() {parameterName=EdgeColor, defaultValue=<rgba 1.5 1.5 1.5 1>}.RGB
    TintMask
   ) {materialFunction=/Engine/Functions/MaterialLayerFunctions/MatLayerBlend_Tint}.Blended Material
   Alpha=VertexColor().R
  ) {materialFunction=&MatLayerBlend_Standard}
  n19 = TextureSampleParameter2D() {group=Rock, parameterName=RockSpec, texture=/Game/Environments/Props/Rocks/Textures/T_Lava_Cliff_AD_S}
  n20 = TextureSampleParameter2D() {group=Rock, parameterName=RockNormal, samplerType=SAMPLERTYPE_Normal, texture=/Game/Environments/Props/Rocks/Textures/T_Lava_Cliff_AD_N}
  n21 = MakeMaterialAttributes!(
   BaseColor=Multiply(
    A=ScalarParameter() {defaultValue=0.6952, group=Rock, parameterName=RockLayerBrightness}
    B=TextureSampleParameter2D() {group=Rock, parameterName=RockDiffuse, texture=/Game/Environments/Props/Rocks/Textures/T_Lava_Cliff_AD_D}.RGB
   )
   Metallic
   Specular=Multiply(A=ScalarParameter() {defaultValue=3.0564, group=Rock, parameterName=Specular_RockBlend}, B=@n19.R)
   Roughness=Multiply(A=ScalarParameter() {defaultValue=3.6856, group=Rock, parameterName=Roughness_RockBlend}, B=@n19.B)
   Anisotropy
   EmissiveColor
   Opacity
   OpacityMask
   Normal=Transform(None=@n20.RGB)
   Tangent
   WorldPositionOffset
   SubsurfaceColor
   ClearCoat
   ClearCoatRoughness
   AmbientOcclusion
   Refraction
   CustomizedUVs_0
   CustomizedUVs_1
   CustomizedUVs_2
   CustomizedUVs_3
   CustomizedUVs_4
   CustomizedUVs_5
   CustomizedUVs_6
   CustomizedUVs_7
   PixelDepthOffset
   ShadingModel
   Displacement
  )
  n22 = ObjectPositionWS()
  n23 = Subtract(A=WorldPosition().XYZ, B=@n22)
  n24 = Frac(None=Divide(A=ComponentMask(None=@n22) {b=true}) {constB=384})
  n25 = Multiply(A=@n24) {constB=384}
  n26 = ScalarParameter() {defaultValue=-322.1895, parameterName=EdgeGradientOffset}
  n27 = Divide(
   A=Add(A=ComponentMask(None=@n23) {b=true}, B=Add(A=@n25, B=@n26))
   B=ScalarParameter() {defaultValue=129.9158, parameterName=EdgeGradientLength}
  )
  n28 = Multiply(A=OneMinus(None=Clamp(None=@n27)), B=VertexColor().R)
  n29 = ComponentMask(None=TextureCoordinate() {coordinateIndex=1}) {g=true}
  n30 = StaticSwitchParameter(
   True=MaterialFunctionCall!(In=@n17, Height=@n4.Height, UVs, TransitionGradient=StaticSwitchParameter(True=@n28, False=@n29) {parameterName=UseRockBlend}) {materialFunction=/Game/Environments/Sets/Vulcano/Materials/MF_LavaEdgeGradient}.Result
   False=@n17
  ) {parameterName=UseEdgeGradient}
  n31 = SetMaterialAttributes!(
   MaterialAttributes=@n30
   Normal=MaterialFunctionCall(
    BaseNormal=GetMaterialAttributes(None=@n21) {attributeGetTypes [#2]}.Normal
    AdditionalNormal=GetMaterialAttributes(None=@n30) {attributeGetTypes [#2]}.Normal
   ) {materialFunction=&BlendAngleCorrectedNormals}.Result
  ) {attributeSetTypes [#2]}
  n32 = StaticSwitchParameter(True=@n16, False=VertexColor().G) {parameterName=UseWallBlendForRockBake}
  MaterialFunctionCall(MA In=StaticSwitchParameter(
    True=MaterialFunctionCall(Base Material=@n21, Top Material=@n31, Alpha=@n32) {materialFunction=&MatLayerBlend_Standard}.Blended Material
    False=@n30
   ) {parameterName=UseRockBlend}) {materialFunction=&MF_CreativeMoveTool}
  MaterialFunctionCall() {materialFunction=&MF_LavaGround}
  Min(A=OneMinus(None=@n24), B=@n24)
 ]
}
