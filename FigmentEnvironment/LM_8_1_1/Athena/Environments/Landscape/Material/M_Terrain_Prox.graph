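# M_Terrain_Prox is a minimal Default Lit landscape material in the FigmentEnvironment content, with no textures or parameters.
# A Landscape Layer Blend node declares a single weight-blended layer named Test, fed by a constant value of 1.
# No connection from the blend to a material output is shown, so the graph holds little beyond that layer definition.
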Material /FigmentEnvironment/LM_8_1_1/Athena/Environments/Landscape/Material/M_Terrain_Prox
graph {
 properties {pixelDepthOffsetMode=PDOM_Legacy, usedShadingModels=Default Lit}
 nodes [
  MaterialExpressionConstant_0 = Constant() {r=1}
  LandscapeLayerBlend!(Layer Test=@MaterialExpressionConstant_0) {layers [{
     blendType=LB_WeightBlend
     constHeightInput=0
     layerName=Test
     previewWeight=0
     constLayerInput=<0 0 0>
     heightInput {expression=null, inputName=None, mask=0, maskA=0, maskB=0, maskG=0, maskR=0, outputIndex=0}
     layerInput {
      expression=&self:MaterialExpressionConstant_0
      inputName=None
      mask=0
      maskA=0
      maskB=0
      maskG=0
      maskR=0
      outputIndex=0
     }
    }]}
 ]
}
